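# Update-required login dialog widget, built on FortUpdateRequiredWidget, showing a title, scrollable description and Update button over a dimmed background within a safe zone.
# AdaptToDeviceOrientation stretches the button to fill width when the device is portrait and the viewport height/width ratio exceeds a threshold, otherwise left-aligns it.
# It runs on Construct and on PlatformGameInstance's screen-orientation-changed delegate, which is unbound on Destruct.
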WidgetBlueprint /FortniteCoreUI/UI/Login/UpdateRequired/UpdateRequiredMP
graphs [
 AdaptToDeviceOrientation {nodes [
   AdaptToDeviceOrientation() {is=title K2Node_FunctionEntry}
   n1 = OR Boolean(
    A="Equal (Enum)"(A:EScreenOrientation Enum=GetDeviceOrientation!() -> (ReturnValue:EScreenOrientation Enum=Unknown), B:EScreenOrientation Enum=Portrait)
    B="Equal (Enum)"(A=GetDeviceOrientation() -> (ReturnValue=Unknown), B=PortraitUpsideDown)
   )
   n2 = GetViewportSize!() -> (ReturnValue_X:double, ReturnValue_Y:double)
   n3 = "float > float"(A=float / float(A=@n2.ReturnValue_Y, B=@n2), B=1.34)
   > SetHorizontalAlignment(
    self:Vertical Box Slot Object Reference=SlotAsVerticalBoxSlot(Widget=Get Button_SizeBox() -> (Button_SizeBox:Size Box Object Reference))
    InHorizontalAlignment=Select(Option 0:EHorizontalAlignment Enum=HAlign_Left, Option 1:EHorizontalAlignment Enum=HAlign_Fill, Index=AND Boolean(A=@n1, B=@n3)) -> (ReturnValue:EHorizontalAlignment Enum)
   )
  ]}
 EventGraph {nodes [
   n1 = OnOrientationChanged!() -> (OutputDelegate, then, inScreenOrientation:EScreenOrientation Enum=Unknown) {is=title K2Node_CustomEvent}
   > AdaptToDeviceOrientation() {is=title K2Node_CallFunction}
   Event Construct()
   > AdaptToDeviceOrientation()
   > n2 = Cast To PlatformGameInstance!(execute, Object=GetGameInstance!(self:Widget Object Reference) -> (ReturnValue:Game Instance Object Reference)) -> (then, CastFailed, AsPlatform Game Instance:Platform Game Instance Object Reference) {is=title K2Node_DynamicCast}
   > n3 = Set PlatformGameInstance(PlatformGameInstance:Platform Game Instance Object Reference=@n2) -> (Output_Get:Platform Game Instance Object Reference)
   > Bind Event to Application Received Screen Orientation Changed Notification Delegate(self:Platform Game Instance Object Reference=@n3, Delegate=Create Event()) {is=title K2Node_AddDelegate}
   Event Destruct()
   n4 = Get PlatformGameInstance() -> (PlatformGameInstance:Platform Game Instance Object Reference)
   > Is Valid(InputObject=@n4)
   > Unbind Event from Application Received Screen Orientation Changed Notification Delegate(self:Platform Game Instance Object Reference=@n4, Delegate=@n1) {is=title K2Node_RemoveDelegate}
   Event PreConstruct()
  ]}
]
variables [PlatformGameInstance {type=PlatformGameInstance, default=null}]
widgets {
 info {inheritedWidgetCount=3, parentClass=/Script/FortniteUI.FortUpdateRequiredWidget, rootWidgetClass=&Overlay, widgetCount=19}
 widgets [
  Overlay_0 {widgetClassPath=&Overlay, widget=Overlay_0}
  CommonBorder_85 {
   widgetClassPath=&CommonBorder
   parent=Overlay_0
   widget=CommonBorder_85
   properties {
    visibility=Visible
    background {
     bIsDynamicallyLoaded=false
     drawAs=Image
     imageType=NoImage
     mirroring=NoMirror
     resourceName=None
     resourceObject=null
     tiling=NoTile
     imageSize=<32 32>
     margin=<ltrb 0 0 0 0>
     outlineSettings {
      bUseBrushTransparency=false
      roundingType=HalfHeightRadius
      width=0
      color {colorUseRule=UseColor_Specified, specifiedColor=<rgba 0 0 0 0>}
      cornerRadii=<0 0 0 0>
     }
     tintColor {colorUseRule=UseColor_Specified, specifiedColor=<rgba 0 0 0 0.5>}
     uVRegion {bIsValid=false, max=<0 0>, min=<0 0>}
    }
   }
   slot {class=OverlaySlot, properties {horizontalAlignment=HAlign_Fill, verticalAlignment=VAlign_Fill}}
  }
  SafeZone {
   widgetClassPath=&SafeZone
   parent=Overlay_0
   widget=SafeZone
   properties {padLeft=false, padRight=false}
   slot {class=OverlaySlot, properties {horizontalAlignment=HAlign_Fill, verticalAlignment=VAlign_Fill}}
  }
  CommonBorder_0 {
   widgetClassPath=&CommonBorder
   parent=SafeZone
   widget=CommonBorder_0
   properties {
    background {
     bIsDynamicallyLoaded=false
     drawAs=RoundedBox
     imageType=FullColor
     mirroring=NoMirror
     resourceName=None
     resourceObject=null
     tiling=NoTile
     imageSize=<128 32>
     margin=<ltrb 0.1 0.45 0.9 0.5>
     outlineSettings {
      bUseBrushTransparency=false
      roundingType=FixedRadius
      width=0
      color {colorUseRule=UseColor_Specified, specifiedColor=<rgba 0 0 0 0>}
      cornerRadii=<20 20 20 20>
     }
     tintColor {colorUseRule=UseColor_Specified, specifiedColor=<rgba 0 0 0 0.95>}
     uVRegion {bIsValid=false, max=<0 0>, min=<0 0>}
    }
    brushColor=<rgba 0 0 0 0>
    padding=<ltrb 10 50 10 50>
   }
   slot {class=SafeZoneSlot, properties {hAlign=HAlign_Center, vAlign=VAlign_Center}}
  }
  SizeBox_45 {
   widgetClassPath=&SizeBox
   parent=CommonBorder_0
   widget=SizeBox_45
   properties {
    bOverride_MaxDesiredWidth=true
    bOverride_MinDesiredHeight=true
    bOverride_MinDesiredWidth=true
    maxDesiredWidth=1350
    minDesiredHeight=500
    minDesiredWidth=750
   }
   slot {class=BorderSlot, properties {padding=<ltrb 10 50 10 50>}}
  }
  Overlay_149 {widgetClassPath=&Overlay, parent=SizeBox_45, widget=Overlay_149, slot {class=SizeBoxSlot, properties {padding=<ltrb 30 2 30 2>}}}
  CommonBorder_87 {
   widgetClassPath=&CommonBorder
   parent=Overlay_149
   widget=CommonBorder_87
   properties {background {
     bIsDynamicallyLoaded=false
     drawAs=RoundedBox
     imageType=NoImage
     mirroring=NoMirror
     resourceName=None
     resourceObject=null
     tiling=NoTile
     imageSize=<32 32>
     margin=<ltrb 0 0 0 0>
     outlineSettings {
      bUseBrushTransparency=false
      roundingType=FixedRadius
      width=0
      color {colorUseRule=UseColor_Specified, specifiedColor=<rgba 0 0 0 0>}
      cornerRadii=<15 15 15 15>
     }
     tintColor {colorUseRule=UseColor_Specified, specifiedColor=<rgba 0.0091 0.0144 0.0262 1>}
     uVRegion {bIsValid=false, max=<0 0>, min=<0 0>}
    }}
   slot {class=OverlaySlot, properties {horizontalAlignment=HAlign_Fill, verticalAlignment=VAlign_Fill}}
  }
  VerticalBox_0 {
   widgetClassPath=&VerticalBox
   parent=Overlay_149
   widget=VerticalBox_0
   slot {class=OverlaySlot, properties {horizontalAlignment=HAlign_Fill, verticalAlignment=VAlign_Fill}}
  }
  VerticalBox_1 {
   widgetClassPath=&VerticalBox
   parent=VerticalBox_0
   widget=VerticalBox_1
   slot {class=VerticalBoxSlot, properties {size {sizeRule=Fill, value=1}}}
  }
  SizeBox_47 {
   widgetClassPath=&SizeBox
   parent=VerticalBox_1
   widget=SizeBox_47
   properties {bOverride_MinDesiredHeight=true, minDesiredHeight=100}
   slot {class=VerticalBoxSlot}
  }
  CommonBorder_1 {
   widgetClassPath=&CommonBorder
   parent=SizeBox_47
   widget=CommonBorder_1
   properties {
    horizontalAlignment=HAlign_Center
    verticalAlignment=VAlign_Center
    background {
     bIsDynamicallyLoaded=false
     drawAs=RoundedBox
     imageType=NoImage
     mirroring=NoMirror
     resourceName=None
     resourceObject=null
     tiling=NoTile
     imageSize=<32 32>
     margin=<ltrb 0 0 0 0>
     outlineSettings {
      bUseBrushTransparency=false
      roundingType=FixedRadius
      width=0
      color {colorUseRule=UseColor_Specified, specifiedColor=<rgba 0 0 0 0>}
      cornerRadii=<15 15 0 0>
     }
     tintColor {colorUseRule=UseColor_Specified, specifiedColor=<rgba 0.0048 0.007 0.0103 1>}
     uVRegion {bIsValid=false, max=<0 0>, min=<0 0>}
    }
    padding=<ltrb 75 40 75 40>
   }
   slot {class=SizeBoxSlot}
  }
  ScaleBox_0 {
   widgetClassPath=&ScaleBox
   parent=CommonBorder_1
   widget=ScaleBox_0
   slot {class=BorderSlot, properties {horizontalAlignment=HAlign_Center, verticalAlignment=VAlign_Center, padding=<ltrb 75 40 75 40>}}
  }
  Text_Title {
   widgetClassPath=&CommonTextBlock
   bInherited=true
   parent=ScaleBox_0
   widget=Text_Title
   properties {
    bIsScrollingEnabled=false
    justification=Center
    text="{ Update Required Title }"
    font {
     bForceMonospaced=false
     bMaterialIsStencil=false
     fontMaterial=null
     fontObject=&HeadingNow
     letterSpacing=-40
     monospacedWidth=1
     size=36
     skewAmount=0
     typefaceFontName=76 Bold
     outlineSettings {
      bApplyOutlineToDropShadows=false
      bMiteredCorners=false
      bSeparateFillAlpha=false
      outlineMaterial=null
      outlineSize=0
      outlineColor=<rgba 0 0 0 1>
     }
    }
    shadowOffset=<0 0>
   }
   slot {class=ScaleBoxSlot}
  }
  VerticalBox_2 {
   widgetClassPath=&VerticalBox
   parent=VerticalBox_1
   widget=VerticalBox_2
   slot {class=VerticalBoxSlot, properties {size {sizeRule=Fill, value=1}}}
  }
  ScrollBox_0 {
   widgetClassPath=&ScrollBox
   parent=VerticalBox_2
   widget=ScrollBox_0
   properties {widgetBarStyle {
     thickness=16
     draggedThumbImage {
      bIsDynamicallyLoaded=false
      drawAs=Box
      imageType=FullColor
      mirroring=NoMirror
      resourceName=../../../Engine/Content/Slate/Common/Scrollbar_Thumb.png
      resourceObject=null
      tiling=NoTile
      imageSize=<0 0>
      margin=<ltrb 0.25 0.25 0.25 0.25>
      outlineSettings {
       bUseBrushTransparency=false
       roundingType=HalfHeightRadius
       width=0
       color {colorUseRule=UseColor_Specified, specifiedColor=<rgba 0 0 0 0>}
       cornerRadii=<0 0 0 0>
      }
      tintColor {colorUseRule=UseColor_Specified, specifiedColor=<rgba 1 1 1 1>}
      uVRegion {bIsValid=false, max=<0 0>, min=<0 0>}
     }
     horizontalBackgroundImage {
      bIsDynamicallyLoaded=false
      drawAs=NoDrawType
      imageType=NoImage
      mirroring=NoMirror
      resourceName=None
      resourceObject=null
      tiling=NoTile
      imageSize=<0 0>
      margin=<ltrb 0 0 0 0>
      outlineSettings {
       bUseBrushTransparency=false
       roundingType=HalfHeightRadius
       width=0
       color {colorUseRule=UseColor_Specified, specifiedColor=<rgba 0 0 0 0>}
       cornerRadii=<0 0 0 0>
      }
      tintColor {colorUseRule=UseColor_Specified, specifiedColor=<rgba 1 1 1 1>}
      uVRegion {bIsValid=false, max=<0 0>, min=<0 0>}
     }
     horizontalBottomSlotImage {
      bIsDynamicallyLoaded=false
      drawAs=Image
      imageType=FullColor
      mirroring=NoMirror
      resourceName=../../../Engine/Content/Slate/Common/Scrollbar_Background_Horizontal.png
      resourceObject=null
      tiling=NoTile
      imageSize=<8 8>
      margin=<ltrb 0 0 0 0>
      outlineSettings {
       bUseBrushTransparency=false
       roundingType=HalfHeightRadius
       width=0
       color {colorUseRule=UseColor_Specified, specifiedColor=<rgba 0 0 0 0>}
       cornerRadii=<0 0 0 0>
      }
      tintColor {colorUseRule=UseColor_Specified, specifiedColor=<rgba 1 1 1 1>}
      uVRegion {bIsValid=false, max=<0 0>, min=<0 0>}
     }
     horizontalTopSlotImage {
      bIsDynamicallyLoaded=false
      drawAs=Image
      imageType=FullColor
      mirroring=NoMirror
      resourceName=../../../Engine/Content/Slate/Common/Scrollbar_Background_Horizontal.png
      resourceObject=null
      tiling=NoTile
      imageSize=<8 8>
      margin=<ltrb 0 0 0 0>
      outlineSettings {
       bUseBrushTransparency=false
       roundingType=HalfHeightRadius
       width=0
       color {colorUseRule=UseColor_Specified, specifiedColor=<rgba 0 0 0 0>}
       cornerRadii=<0 0 0 0>
      }
      tintColor {colorUseRule=UseColor_Specified, specifiedColor=<rgba 1 1 1 1>}
      uVRegion {bIsValid=false, max=<0 0>, min=<0 0>}
     }
     hoveredThumbImage {
      bIsDynamicallyLoaded=false
      drawAs=Box
      imageType=FullColor
      mirroring=NoMirror
      resourceName=../../../Engine/Content/Slate/Common/Scrollbar_Thumb.png
      resourceObject=null
      tiling=NoTile
      imageSize=<0 0>
      margin=<ltrb 0.25 0.25 0.25 0.25>
      outlineSettings {
       bUseBrushTransparency=false
       roundingType=HalfHeightRadius
       width=0
       color {colorUseRule=UseColor_Specified, specifiedColor=<rgba 0 0 0 0>}
       cornerRadii=<0 0 0 0>
      }
      tintColor {colorUseRule=UseColor_Specified, specifiedColor=<rgba 1 1 1 1>}
      uVRegion {bIsValid=false, max=<0 0>, min=<0 0>}
     }
     normalThumbImage {
      bIsDynamicallyLoaded=false
      drawAs=Box
      imageType=FullColor
      mirroring=NoMirror
      resourceName=../../../Engine/Content/Slate/Common/Scrollbar_Thumb.png
      resourceObject=null
      tiling=NoTile
      imageSize=<0 0>
      margin=<ltrb 0.25 0.25 0.25 0.25>
      outlineSettings {
       bUseBrushTransparency=false
       roundingType=HalfHeightRadius
       width=0
       color {colorUseRule=UseColor_Specified, specifiedColor=<rgba 0 0 0 0>}
       cornerRadii=<0 0 0 0>
      }
      tintColor {colorUseRule=UseColor_Specified, specifiedColor=<rgba 1 1 1 1>}
      uVRegion {bIsValid=false, max=<0 0>, min=<0 0>}
     }
     verticalBackgroundImage {
      bIsDynamicallyLoaded=false
      drawAs=Image
      imageType=NoImage
      mirroring=NoMirror
      resourceName=None
      resourceObject=null
      tiling=NoTile
      imageSize=<0 0>
      margin=<ltrb 0 0 0 0>
      outlineSettings {
       bUseBrushTransparency=false
       roundingType=HalfHeightRadius
       width=0
       color {colorUseRule=UseColor_Specified, specifiedColor=<rgba 0 0 0 0>}
       cornerRadii=<0 0 0 0>
      }
      tintColor {colorUseRule=UseColor_Specified, specifiedColor=<rgba 1 1 1 0.5>}
      uVRegion {bIsValid=false, max=<0 0>, min=<0 0>}
     }
     verticalBottomSlotImage {
      bIsDynamicallyLoaded=false
      drawAs=Image
      imageType=FullColor
      mirroring=NoMirror
      resourceName=../../../Engine/Content/Slate/Common/Scrollbar_Background_Vertical.png
      resourceObject=null
      tiling=NoTile
      imageSize=<8 8>
      margin=<ltrb 0 0 0 0>
      outlineSettings {
       bUseBrushTransparency=false
       roundingType=HalfHeightRadius
       width=0
       color {colorUseRule=UseColor_Specified, specifiedColor=<rgba 0 0 0 0>}
       cornerRadii=<0 0 0 0>
      }
      tintColor {colorUseRule=UseColor_Specified, specifiedColor=<rgba 1 1 1 1>}
      uVRegion {bIsValid=false, max=<0 0>, min=<0 0>}
     }
     verticalTopSlotImage {
      bIsDynamicallyLoaded=false
      drawAs=Image
      imageType=FullColor
      mirroring=NoMirror
      resourceName=../../../Engine/Content/Slate/Common/Scrollbar_Background_Vertical.png
      resourceObject=null
      tiling=NoTile
      imageSize=<8 8>
      margin=<ltrb 0 0 0 0>
      outlineSettings {
       bUseBrushTransparency=false
       roundingType=HalfHeightRadius
       width=0
       color {colorUseRule=UseColor_Specified, specifiedColor=<rgba 0 0 0 0>}
       cornerRadii=<0 0 0 0>
      }
      tintColor {colorUseRule=UseColor_Specified, specifiedColor=<rgba 1 1 1 1>}
      uVRegion {bIsValid=false, max=<0 0>, min=<0 0>}
     }
    }}
   slot {class=VerticalBoxSlot, properties {padding=<ltrb 75 60 40 20>, size {sizeRule=Fill, value=1}}}
  }
  Text_Description {
   widgetClassPath=&CommonTextBlock
   bInherited=true
   parent=ScrollBox_0
   widget=Text_Description
   properties {
    autoWrapText=true
    lineHeightPercentage=1.2
    text="{ Update Required Description }"
    font {
     bForceMonospaced=false
     bMaterialIsStencil=false
     fontMaterial=null
     fontObject=&HeadingNow
     letterSpacing=0
     monospacedWidth=1
     size=31.5
     skewAmount=0
     typefaceFontName=74 Regular
     outlineSettings {
      bApplyOutlineToDropShadows=false
      bMiteredCorners=false
      bSeparateFillAlpha=false
      outlineMaterial=null
      outlineSize=0
      outlineColor=<rgba 0 0 0 1>
     }
    }
    shadowOffset=<0 0>
   }
   slot {class=ScrollBoxSlot, properties {padding=<ltrb 0 0 35 0>}}
  }
  Button_SizeBox {
   widgetClassPath=&SizeBox
   bIsVariable=true
   parent=VerticalBox_0
   widget=Button_SizeBox
   slot {class=VerticalBoxSlot, properties {horizontalAlignment=HAlign_Left, verticalAlignment=VAlign_Bottom, padding=<ltrb 75 30 75 50>}}
  }
  Overlay_Button_Confirm {widgetClassPath=&Overlay, bIsVariable=true, parent=Button_SizeBox, widget=Overlay_Button_Confirm, slot {class=SizeBoxSlot}}
  Button_Confirm {
   widgetClassPath=/FortUILibrary/UIKit/Buttons/WBP_UIKit_Button_Loud
   bInherited=true
   bIsVariable=true
   parent=Overlay_Button_Confirm
   widget=Button_Confirm
   properties {buttonSize=Large, text=Update, triggeringInputAction {dataTable=&InputActionDataTable, rowName=UpdateNow}}
   slot {class=OverlaySlot, properties {horizontalAlignment=HAlign_Fill, verticalAlignment=VAlign_Fill}}
  }
 ]
}
