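# Widget Blueprint for the controller-disconnected modal, laying out the UI for native parent AthenaControllerDisconnectedModal.
# A dimming Lightbox hosts a centered column with placeholder title and description, a progress spinner, and mirrored skewed separators.
# An initially collapsed Change User Account row pairs its label with a GenericForward action icon.
# On activation, the graph only plays Intro, a short opacity fade-in of the root overlay.
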WidgetBlueprint /Game/UI/Foundation/Modals/ControllerDisconnectedModal/Athena_ControllerDisconnectedModal
animations [Intro {
  binding_count=1
  length_seconds=0.25
  object=&self:Intro
  sequence {bindings [Overlay_0 {
     id=#1
     possessed_class=Overlay
     tracks [{
       type=MovieSceneFloatTrack
       display_name=Render Opacity
       sections [{type=MovieSceneFloatSection, channels [None {type=MovieSceneScriptingFloatChannel, keys [[0,0],[5,1]]}]}]
      }]
    }]}
 }]
graphs [EventGraph {nodes [
   Event On Activated()
   > PlayAnimationForward(InAnimation=Get Intro())
  ]}]
variables []
widgets {
 info {
  inheritedWidgetCount=4
  namedSlotCount=1
  parentClass=/Script/FortniteUI.AthenaControllerDisconnectedModal
  rootWidgetClass=&Overlay
  widgetCount=19
 }
 widgets [
  Overlay_0 {widgetClassPath=&Overlay, widget=Overlay_0}
  Lightbox {
   widgetClassPath=/Game/UI/Foundation/Lightbox/Lightbox
   bIsVariable=true
   parent=Overlay_0
   widget=Lightbox
   properties {dimAmount=0.8}
   slot {class=OverlaySlot, properties {horizontalAlignment=HAlign_Fill, verticalAlignment=VAlign_Fill}}
  }
  Overlay_3 {widgetClassPath=&Overlay, namedSlotHost=Lightbox, widget=Overlay_3}
  Image_66 {
   widgetClassPath=&Image
   bIsVariable=true
   parent=Overlay_3
   widget=Image_66
   properties {
    visibility=HitTestInvisible
    brush {
     bIsDynamicallyLoaded=false
     drawAs=Image
     imageType=NoImage
     mirroring=NoMirror
     resourceName=None
     resourceObject=/FortUILibrary/Styles/Border/Materials/M_Border_SoftShadow_DarkBLue
     tiling=NoTile
     imageSize=<32 32>
     margin=<ltrb 0 0 0 0>
     outlineSettings {
      bUseBrushTransparency=false
      roundingType=HalfHeightRadius
      width=0
      color {colorUseRule=UseColor_Specified, specifiedColor=<rgba 0 0 0 0>}
      cornerRadii=<0 0 0 0>
     }
     tintColor {colorUseRule=UseColor_Specified, specifiedColor=<rgba 1 1 1 1>}
     uVRegion {bIsValid=false, max=<0 0>, min=<0 0>}
    }
    colorAndOpacity=<rgba 1 1 1 0.5>
   }
   slot {class=OverlaySlot, properties {horizontalAlignment=HAlign_Fill, verticalAlignment=VAlign_Fill}}
  }
  SizeBox_0 {
   widgetClassPath=&SizeBox
   parent=Overlay_3
   widget=SizeBox_0
   properties {
    bOverride_MaxDesiredHeight=true
    bOverride_MaxDesiredWidth=true
    bOverride_WidthOverride=true
    heightOverride=800
    maxDesiredHeight=800
    maxDesiredWidth=512
    minDesiredHeight=512
    widthOverride=1075
   }
   slot {class=OverlaySlot, properties {horizontalAlignment=HAlign_Center, verticalAlignment=VAlign_Center, padding=<ltrb 0 150 0 150>}}
  }
  CommonBorder_15 {
   widgetClassPath=&CommonBorder
   parent=SizeBox_0
   widget=CommonBorder_15
   properties {
    visibility=Visible
    background {
     bIsDynamicallyLoaded=false
     drawAs=NoDrawType
     imageType=NoImage
     mirroring=NoMirror
     resourceName=None
     resourceObject=null
     tiling=NoTile
     imageSize=<128 32>
     margin=<ltrb 0 0 0 0>
     outlineSettings {
      bUseBrushTransparency=false
      roundingType=HalfHeightRadius
      width=0
      color {colorUseRule=UseColor_Specified, specifiedColor=<rgba 0 0 0 0>}
      cornerRadii=<0 0 0 0>
     }
     tintColor {colorUseRule=UseColor_Specified, specifiedColor=<rgba 1 1 1 1>}
     uVRegion {bIsValid=false, max=<0 0>, min=<0 0>}
    }
    padding=<ltrb 0 0 0 0>
   }
   slot {class=SizeBoxSlot}
  }
  MainBox {
   widgetClassPath=&VerticalBox
   parent=CommonBorder_15
   widget=MainBox
   properties {navigation=&self:WidgetTree.MainBox.WidgetNavigation_0}
   slot {class=BorderSlot, properties {padding=<ltrb 0 0 0 0>}}
  }
  CommonBorder_Modal {
   widgetClassPath=&CommonBorder
   parent=MainBox
   widget=CommonBorder_Modal
   properties {background {
     bIsDynamicallyLoaded=false
     drawAs=NoDrawType
     imageType=NoImage
     mirroring=NoMirror
     resourceName=None
     resourceObject=null
     tiling=NoTile
     imageSize=<64 64>
     margin=<ltrb 0.25 0.25 0.25 0.25>
     outlineSettings {
      bUseBrushTransparency=false
      roundingType=HalfHeightRadius
      width=0
      color {colorUseRule=UseColor_Specified, specifiedColor=<rgba 0 0 0 0>}
      cornerRadii=<0 0 0 0>
     }
     tintColor {colorUseRule=UseColor_Specified, specifiedColor=<rgba 1 1 1 1>}
     uVRegion {bIsValid=false, max=<0 0>, min=<0 0>}
    }}
   slot {class=VerticalBoxSlot}
  }
  VerticalBox_2 {widgetClassPath=&VerticalBox, parent=CommonBorder_Modal, widget=VerticalBox_2, slot {class=BorderSlot}}
  Text_Title {
   widgetClassPath=&CommonTextBlock
   bInherited=true
   parent=VerticalBox_2
   widget=Text_Title
   properties {
    autoWrapText=true
    clipping=ClipToBounds
    justification=Center
    lineHeightPercentage=0.9
    scrollStyle=&TextScrollStyle-Base_C
    text="{TITLE}"
    textTransformPolicy=ToUpper
    font {
     bForceMonospaced=false
     bMaterialIsStencil=false
     fontMaterial=null
     fontObject=&Burbank
     letterSpacing=0
     monospacedWidth=1
     size=48
     skewAmount=0
     typefaceFontName=BurbankRegular-Black
     outlineSettings {
      bApplyOutlineToDropShadows=false
      bMiteredCorners=false
      bSeparateFillAlpha=false
      outlineMaterial=null
      outlineSize=0
      outlineColor=<rgba 0 0 0 1>
     }
    }
    margin=<ltrb 0 0 0 10>
    renderTransform {angle=0, scale=<1 1>, shear=<-12 0>, translation=<0 0>}
    shadowOffset=<0 0>
   }
   slot {class=VerticalBoxSlot, properties {padding=<ltrb 0 0 0 8>, size {sizeRule=Fill, value=1}}}
  }
  Image_Separator_Up {
   widgetClassPath=&Image
   bIsVariable=true
   parent=MainBox
   widget=Image_Separator_Up
   properties {
    brush {
     bIsDynamicallyLoaded=false
     drawAs=Box
     imageType=NoImage
     mirroring=NoMirror
     resourceName=None
     resourceObject=/Game/Athena/UI/ReplayBrowser/M_Ui_Skewed_Vfor9slices_Inst1
     tiling=NoTile
     imageSize=<16 16>
     margin=<ltrb 0.5 0.5 0.5 0.5>
     outlineSettings {
      bUseBrushTransparency=false
      roundingType=HalfHeightRadius
      width=0
      color {colorUseRule=UseColor_Specified, specifiedColor=<rgba 0 0 0 0>}
      cornerRadii=<0 0 0 0>
     }
     tintColor {colorUseRule=UseColor_Specified, specifiedColor=<rgba 1 1 1 1>}
     uVRegion {bIsValid=false, max=<0 0>, min=<0 0>}
    }
    colorAndOpacity=<rgba 0 0.0908 1 1>
   }
   slot {class=VerticalBoxSlot, properties {verticalAlignment=VAlign_Top}}
  }
  VerticalBox_0 {
   widgetClassPath=&VerticalBox
   parent=MainBox
   widget=VerticalBox_0
   slot {class=VerticalBoxSlot, properties {padding=<ltrb 0 5 0 5>}}
  }
  Text_Description {
   widgetClassPath=&CommonTextBlock
   bInherited=true
   parent=VerticalBox_0
   widget=Text_Description
   properties {
    autoWrapText=true
    clipping=ClipToBounds
    justification=Center
    lineHeightPercentage=0.9
    text="{DESCRIPTION}"
    textTransformPolicy=ToUpper
    font {
     bForceMonospaced=false
     bMaterialIsStencil=false
     fontMaterial=null
     fontObject=&BurbankSmall
     letterSpacing=0
     monospacedWidth=1
     size=22
     skewAmount=0
     typefaceFontName=Bold
     outlineSettings {
      bApplyOutlineToDropShadows=false
      bMiteredCorners=false
      bSeparateFillAlpha=false
      outlineMaterial=null
      outlineSize=0
      outlineColor=<rgba 0 0 0 1>
     }
    }
    margin=<ltrb 0 0 0 10>
    renderTransform {angle=0, scale=<1 1>, shear=<-12 0>, translation=<0 0>}
    shadowOffset=<0 0>
   }
   slot {class=VerticalBoxSlot, properties {padding=<ltrb 0 0 0 8>, size {sizeRule=Fill, value=1}}}
  }
  ProgressSpinner {
   widgetClassPath=&Image
   bIsVariable=true
   parent=VerticalBox_0
   widget=ProgressSpinner
   properties {brush {
     bIsDynamicallyLoaded=false
     drawAs=Image
     imageType=NoImage
     mirroring=NoMirror
     resourceName=None
     resourceObject=/FortUILibrary/UIKit/Throbber/MI_UIKit_Color_Throbber
     tiling=NoTile
     imageSize=<64 64>
     margin=<ltrb 0 0 0 0>
     outlineSettings {
      bUseBrushTransparency=false
      roundingType=HalfHeightRadius
      width=0
      color {colorUseRule=UseColor_Specified, specifiedColor=<rgba 0 0 0 0>}
      cornerRadii=<0 0 0 0>
     }
     tintColor {colorUseRule=UseColor_Specified, specifiedColor=<rgba 1 1 1 1>}
     uVRegion {bIsValid=false, max=<0 0>, min=<0 0>}
    }}
   slot {class=VerticalBoxSlot, properties {horizontalAlignment=HAlign_Center, verticalAlignment=VAlign_Center}}
  }
  Image_Separator_Down {
   widgetClassPath=&Image
   bIsVariable=true
   parent=MainBox
   widget=Image_Separator_Down
   properties {
    brush {
     bIsDynamicallyLoaded=false
     drawAs=Box
     imageType=NoImage
     mirroring=NoMirror
     resourceName=None
     resourceObject=&M_Ui_Skewed_Vfor9slices_Inst1
     tiling=NoTile
     imageSize=<16 16>
     margin=<ltrb 0.5 0.5 0.5 0.5>
     outlineSettings {
      bUseBrushTransparency=false
      roundingType=HalfHeightRadius
      width=0
      color {colorUseRule=UseColor_Specified, specifiedColor=<rgba 0 0 0 0>}
      cornerRadii=<0 0 0 0>
     }
     tintColor {colorUseRule=UseColor_Specified, specifiedColor=<rgba 1 1 1 1>}
     uVRegion {bIsValid=false, max=<0 0>, min=<0 0>}
    }
    colorAndOpacity=<rgba 0 0.0908 1 1>
    renderTransform {angle=0, scale=<-1 1>, shear=<0 0>, translation=<0 0>}
   }
   slot {class=VerticalBoxSlot, properties {verticalAlignment=VAlign_Bottom}}
  }
  HBox_SwitchUser {
   widgetClassPath=&HorizontalBox
   bInherited=true
   parent=MainBox
   widget=HBox_SwitchUser
   properties {visibility=Collapsed}
   slot {class=VerticalBoxSlot, properties {horizontalAlignment=HAlign_Center, verticalAlignment=VAlign_Bottom}}
  }
  CommonActionWidget_164 {
   widgetClassPath=&CommonActionWidget
   bIsVariable=true
   parent=HBox_SwitchUser
   widget=CommonActionWidget_164
   properties {inputActions [{dataTable=&InputActionDataTable, rowName=GenericForward}]}
   slot {class=HorizontalBoxSlot}
  }
  CommonTextBlock_113 {
   widgetClassPath=&CommonTextBlock
   parent=HBox_SwitchUser
   widget=CommonTextBlock_113
   properties {
    text=Change User Account
    font {
     bForceMonospaced=false
     bMaterialIsStencil=false
     fontMaterial=null
     fontObject=&BurbankSmall
     letterSpacing=0
     monospacedWidth=1
     size=22
     skewAmount=0
     typefaceFontName=Bold
     outlineSettings {
      bApplyOutlineToDropShadows=false
      bMiteredCorners=false
      bSeparateFillAlpha=false
      outlineMaterial=null
      outlineSize=0
      outlineColor=<rgba 0 0 0 1>
     }
    }
    renderTransform {angle=0, scale=<1 1>, shear=<-12 0>, translation=<0 0>}
    shadowOffset=<0 0>
   }
   slot {class=HorizontalBoxSlot}
  }
  ActionButton_Cancel {widgetClassPath=/Script/CommonUI.CommonBoundActionButton, bInherited=true, widget=null}
 ]
}
