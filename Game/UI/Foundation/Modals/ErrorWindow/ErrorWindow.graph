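# Error modal widget on FortErrorWindow showing a severity icon, title, scrollable message, error codes and a Continue button at Error input priority.
# Activation plays an intro animation and alert sound; non-mobile platforms also start decorative llama animations.
# A timer and animation-finished event loop recolors the llamas, plays one of three slide animations at random, then reverses it after a random delay.
# Destruct clears both timers.
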WidgetBlueprint /Game/UI/Foundation/Modals/ErrorWindow/ErrorWindow
animations [
 Llama_1 {
  binding_count=1
  display_label=""
  length_seconds=4
  object=&self:Llama_1
  sequence {bindings [ErrorLlama_0 {
     id=#1
     possessed_class=ErrorLlama_C
     tracks [{
       type=MovieScene2DTransformTrack
       sections [{
         type=MovieScene2DTransformSection
         end_frame=80
         start_frame=0
         channels [
          Translation.X {type=MovieSceneScriptingFloatChannel, const [163,2]}
          Translation.Y {type=MovieSceneScriptingFloatChannel, keys [[0,423.6797],[80,-168]]}
          Angle {type=MovieSceneScriptingFloatChannel}
          Scale.X {type=MovieSceneScriptingFloatChannel}
          Scale.Y {type=MovieSceneScriptingFloatChannel}
          Shear.X {type=MovieSceneScriptingFloatChannel}
          Shear.Y {type=MovieSceneScriptingFloatChannel}
         ]
        }]
      }]
    }]}
 }
 Intro_1 {
  binding_count=5
  display_label=""
  length_seconds=0.3
  object=&self:Intro_1
  sequence {bindings [
    VerticalBox_0 {
     id=#2
     possessed_class=VerticalBox
     tracks [{
       type=MovieSceneFloatTrack
       display_name=Render Opacity
       sections [{type=MovieSceneFloatSection, channels [None {type=MovieSceneScriptingFloatChannel, keys [[0,0],[2,0],[5,1]]}]}]
      }]
    }
    ScrollBox_MainError {
     id=#3
     possessed_class=ScrollBox
     tracks [{
       type=MovieSceneByteTrack
       display_name=Scroll Bar Visibility
       sections [{type=MovieSceneByteSection, channels [None {type=MovieSceneScriptingByteChannel, keys [[4,1],[5,0]]}]}]
      }]
    }
    WidgetSwitcher_SeverityImage {
     id=#4
     possessed_class=WidgetSwitcher
     tracks [{
       type=MovieScene2DTransformTrack
       sections [{
         type=MovieScene2DTransformSection
         channels [
          Translation.X {type=MovieSceneScriptingFloatChannel}
          Translation.Y {type=MovieSceneScriptingFloatChannel, keys [[0,64],[2,0]]}
          Angle {type=MovieSceneScriptingFloatChannel}
          Scale.X {type=MovieSceneScriptingFloatChannel, keys [[0,3],[3,0.7702],[5,1]]}
          Scale.Y {type=MovieSceneScriptingFloatChannel, keys [[0,3],[3,0.7702],[5,1]]}
          Shear.X {type=MovieSceneScriptingFloatChannel}
          Shear.Y {type=MovieSceneScriptingFloatChannel}
         ]
        }]
      }]
    }
    SizeBox_3 {
     id=#5
     possessed_class=SizeBox
     tracks [{
       type=MovieSceneFloatTrack
       display_name=Max Desired Height
       sections [{type=MovieSceneFloatSection, channels [None {type=MovieSceneScriptingFloatChannel, keys [[2,0],[5,450]]}]}]
      }]
    }
    CommonButton_Close {
     id=#6
     possessed_class=IconTextButton_C
     tracks [{
       type=MovieScene2DTransformTrack
       sections [{
         type=MovieScene2DTransformSection
         channels [
          Translation.X {type=MovieSceneScriptingFloatChannel}
          Translation.Y {type=MovieSceneScriptingFloatChannel}
          Angle {type=MovieSceneScriptingFloatChannel}
          Scale.X {type=MovieSceneScriptingFloatChannel, keys [[3,0],[5,1.1],[6,1]]}
          Scale.Y {type=MovieSceneScriptingFloatChannel, keys [[3,0],[5,1.1],[6,1]]}
          Shear.X {type=MovieSceneScriptingFloatChannel}
          Shear.Y {type=MovieSceneScriptingFloatChannel}
         ]
        }]
      }]
    }
   ]}
 }
 Llama_2 {
  binding_count=1
  display_label=""
  length_seconds=4
  object=&self:Llama_2
  sequence {bindings [ErrorLlama_C_0 {
     id=#7
     possessed_class=ErrorLlama_C
     tracks [{
       type=MovieScene2DTransformTrack
       sections [{
         type=MovieScene2DTransformSection
         end_frame=80
         start_frame=0
         channels [
          Translation.X {type=MovieSceneScriptingFloatChannel, const [433,2]}
          Translation.Y {type=MovieSceneScriptingFloatChannel, keys [[0,-406],[80,80]]}
          Angle {type=MovieSceneScriptingFloatChannel}
          Scale.X {type=MovieSceneScriptingFloatChannel}
          Scale.Y {type=MovieSceneScriptingFloatChannel}
          Shear.X {type=MovieSceneScriptingFloatChannel}
          Shear.Y {type=MovieSceneScriptingFloatChannel}
         ]
        }]
      }]
    }]}
 }
 Llama_3 {
  binding_count=1
  display_label=""
  length_seconds=4
  object=&self:Llama_3
  sequence {bindings [ErrorLlama_C_1 {
     id=#8
     possessed_class=ErrorLlama_C
     tracks [{
       type=MovieScene2DTransformTrack
       sections [{
         type=MovieScene2DTransformSection
         end_frame=80
         start_frame=0
         channels [
          Translation.X {type=MovieSceneScriptingFloatChannel, keys [[0,489],[80,-42]]}
          Translation.Y {type=MovieSceneScriptingFloatChannel, const [0,2]}
          Angle {type=MovieSceneScriptingFloatChannel}
          Scale.X {type=MovieSceneScriptingFloatChannel}
          Scale.Y {type=MovieSceneScriptingFloatChannel}
          Shear.X {type=MovieSceneScriptingFloatChannel}
          Shear.Y {type=MovieSceneScriptingFloatChannel}
         ]
        }]
      }]
    }]}
 }
]
graphs [
 EventGraph {nodes [
   ResetLlamas() {is=title K2Node_CustomEvent}
   n1 = Get Llama_1() -> (Llama_1:Widget Animation Object Reference)
   > PlayAnimation(InAnimation=@n1, StartAtTime=GetEndTime(self=@n1), PlayMode=Reverse)
   n2 = Get Llama_2() -> (Llama_2:Widget Animation Object Reference)
   > PlayAnimation(InAnimation=@n2, StartAtTime=GetEndTime(self=@n2), PlayMode=Reverse)
   n3 = Get Llama_3() -> (Llama_3:Widget Animation Object Reference)
   > PlayAnimation(InAnimation=@n3, StartAtTime=GetEndTime(self=@n3), PlayMode=Reverse)
   HandleLlamaAnimFinish() {is=title K2Node_CustomEvent}
   > n4 = Set Timer by Event(Delegate=Create Event(), Time=RandomFloatInRange(Min=4, Max=6), bLooping=false)
   > Set LlamaReverseTimer(LlamaReverseTimer:Timer Handle Structure=@n4) -> (Output_Get:Timer Handle Structure)
   > UnbindFromAnimationFinished!(
    execute
    self:Widget Animation Object Reference=Get CurLlamaAnim() -> (CurLlamaAnim:Widget Animation Object Reference)
    Widget:User Widget Object Reference=Self-Reference()
    Delegate:Delegate=Create Event()
   ) -> (then)
   LlamaReverse() {is=title K2Node_CustomEvent}
   n5 = Get CurLlamaAnim()
   > PlayAnimation(InAnimation=@n5, PlayMode=Reverse)
   > BindToAnimationFinished!(execute, self:Widget Animation Object Reference=@n5, Widget:User Widget Object Reference=Self-Reference(), Delegate=Create Event()) -> (then)
   HandleLlamaCycleFinished() {is=title K2Node_CustomEvent}
   > n6 = UnbindFromAnimationFinished(self=Get CurLlamaAnim(), Widget=Self-Reference(), Delegate=Create Event())
   n7 = Set Timer by Event(execute=[@n6.then, @n9.then], Delegate=Create Event(), Time=RandomFloatInRange(Min=4, Max=7), bLooping=false)
   > Set AmbientLlamaPlayTimer(AmbientLlamaPlayTimer:Timer Handle Structure=@n7) -> (Output_Get:Timer Handle Structure)
   PlayRandomAmbientLlama() {is=title K2Node_CustomEvent}
   > RandomizeColor(self:Error Llama Object Reference=Get LlamaArray() -> (LlamaArray:Array of Error Llama Object References))
   > n8 = Set CurLlamaAnim(CurLlamaAnim:Widget Animation Object Reference=Select!(
     Option 0:Widget Animation Object Reference=Get Llama_1()
     Option 1:Widget Animation Object Reference=Get Llama_2()
     Option 2:Widget Animation Object Reference=Get Llama_3()
     Index:int=RandomIntegerInRange!(Min:int=0, Max:int=2) -> (ReturnValue:int=0)
    ) -> (ReturnValue:Widget Animation Object Reference)) -> (Output_Get:Widget Animation Object Reference)
   > BindToAnimationFinished(self=@n8, Widget=Self-Reference(), Delegate=Create Event())
   > PlayAnimation(InAnimation=Get CurLlamaAnim())
   n9 = StartAmbientLlamaSequence() {is=title K2Node_CustomEvent}
   n10 = Event OnOverrideCloseButtonText!() -> (OutputDelegate, then, OverrideText:"Text (by ref)") {is=title K2Node_Event}
   > n11 = Cast To IconTextButton!(execute, Object=Get CommonButton_Close() -> (CommonButton_Close:Common Button Legacy Object Reference)) -> (then, CastFailed, AsIcon Text Button:Icon Text Button Object Reference) {is=title K2Node_DynamicCast}
   > Set Text!(execute, self:Icon Text Button Object Reference=@n11, ButtonText:Text=@n10.OverrideText) -> (then) {is=title K2Node_CallFunction}
   Event On Activated()
   > Parent: On Activated() {is=title K2Node_CallParentFunction}
   > PlayAnimation(InAnimation=Get Intro_1() -> (Intro_1:Widget Animation Object Reference))
   > n12 = Branch(Condition=IsMobileGame())
   PlaySound2D(execute=[@n13.then, @n12.then], Sound=/Game/Sounds/Fort_UI_Sounds/Gameplay/Fort_Error_Window_Cue)
   ResetLlamas(execute=@n12.else) {is=title K2Node_CallFunction}
   > n13 = StartAmbientLlamaSequence() {is=title K2Node_CallFunction}
   Event Construct()
   > Set LlamaArray(LlamaArray:Array of Error Llama Object References=Make Array!(
     "[0]":Error Llama Object Reference=Get ErrorLlama_0() -> (ErrorLlama_0:Error Llama Object Reference)
     "[1]":Error Llama Object Reference=Get ErrorLlama_C_0() -> (ErrorLlama_C_0:Error Llama Object Reference)
     "[2]":Error Llama Object Reference=Get ErrorLlama_C_1() -> (ErrorLlama_C_1:Error Llama Object Reference)
    ) -> (Array:Array of Error Llama Object References)) -> (Output_Get:Array of Error Llama Object References)
   > Initialize Close Button()
   Event Destruct()
   > Clear and Invalidate Timer by Handle(Handle=Get AmbientLlamaPlayTimer() -> (AmbientLlamaPlayTimer:Timer Handle Structure))
   > Clear and Invalidate Timer by Handle(Handle=Get LlamaReverseTimer() -> (LlamaReverseTimer:Timer Handle Structure))
   Event OnBeginOutro() {is=title K2Node_Event}
   > EndOutro(self:Common Activatable Panel Legacy Object Reference)
   Event OnBeginIntro() {is=title K2Node_Event}
   > EndIntro(self:Common Activatable Panel Legacy Object Reference)
   "On Button Clicked (CommonButton_Close)"!() -> (OutputDelegate, then, Button:Common Button Legacy Object Reference) {is=title K2Node_ComponentBoundEvent}
  ]}
 Initialize Close Button {nodes [
   Initialize Close Button() {is=title K2Node_FunctionEntry}
   > SetInputActionPriority!(
    execute
    self:User Widget Object Reference=Get CommonButton_Close()
    NewPriority:int=GetInputPriority!(
     self:Fort Global UIContext Object Reference=Get FortGlobalUIContext!() -> (ReturnValue:Fort Global UIContext Object Reference) {is=title K2Node_GetBlueprintContextEx}
     Priority:EInputPriority Enum=Error
    ) -> (ReturnValue:int=0)
   ) -> (then)
  ]}
 OnAnalogValueChanged {nodes [
   OnAnalogValueChanged!() -> (then, MyGeometry:Geometry Structure, InAnalogInputEvent:Analog Input Event Structure) {is=title K2Node_FunctionEntry}
   > Return Node!(execute, ReturnValue:Event Reply Structure=Unhandled())
  ]}
]
variables [
 AmbientLlamaPlayTimer {type=TimerHandle}
 CurLlamaAnim {type=WidgetAnimation, default=null}
 ErrorWindowAppearSound {type=SoundBase, default=&Fort_Error_Window_Cue}
 ErrorWindowAwaySound {type=SoundBase, default=null}
 LlamaArray {type="array<ErrorLlama_C>", default []}
 LlamaReverseTimer {type=TimerHandle}
]
widgets {
 info {
  inheritedWidgetCount=9
  namedSlotCount=1
  parentClass=/Script/FortniteUI.FortErrorWindow
  rootWidgetClass=&Overlay
  widgetCount=24
 }
 widgets [
  Overlay_0 {widgetClassPath=&Overlay, widget=Overlay_0}
  ErrorLlama_0 {
   widgetClassPath=&/ErrorLlama
   bIsVariable=true
   parent=Overlay_0
   widget=ErrorLlama_0
   properties {renderTransform {angle=0, scale=<5 5>, shear=<0 0>, translation=<163 423.6797>}}
   slot {class=OverlaySlot, properties {verticalAlignment=VAlign_Bottom}}
  }
  ErrorLlama_C_0 {
   widgetClassPath=&ErrorLlama
   bIsVariable=true
   parent=Overlay_0
   widget=ErrorLlama_C_0
   properties {bFlippedTexture=true, renderTransform {angle=-149, scale=<-4 4>, shear=<0 0>, translation=<433 80>}}
   slot {class=OverlaySlot}
  }
  ErrorLlama_C_1 {
   widgetClassPath=&ErrorLlama
   bIsVariable=true
   parent=Overlay_0
   widget=ErrorLlama_C_1
   properties {bFlippedTexture=true, renderTransform {angle=-46, scale=<-4 4>, shear=<0 0>, translation=<489 0>}}
   slot {class=OverlaySlot, properties {horizontalAlignment=HAlign_Right, verticalAlignment=VAlign_Center}}
  }
  Overlay {
   widgetClassPath=&Overlay
   parent=Overlay_0
   widget=Overlay
   properties {bOverride_Cursor=true}
   slot {class=OverlaySlot, properties {horizontalAlignment=HAlign_Fill, verticalAlignment=VAlign_Fill}}
  }
  WBP_UIKit_Scrim {
   widgetClassPath=/FortUILibrary/Systems/Dialog/WBP_UIKit_Scrim
   bIsVariable=true
   parent=Overlay
   widget=WBP_UIKit_Scrim
   slot {class=OverlaySlot, properties {horizontalAlignment=HAlign_Fill, verticalAlignment=VAlign_Fill}}
  }
  Lightbox {
   widgetClassPath=/Game/UI/Foundation/Lightbox/Lightbox
   bIsVariable=true
   parent=Overlay
   widget=Lightbox
   slot {class=OverlaySlot, properties {horizontalAlignment=HAlign_Fill, verticalAlignment=VAlign_Fill}}
  }
  "[Border[ Dialog Content" {
   widgetClassPath=&Border
   namedSlotHost=Lightbox
   widget="[Border[ Dialog Content"
   properties {
    horizontalAlignment=HAlign_Center
    verticalAlignment=VAlign_Center
    background {
     bIsDynamicallyLoaded=false
     drawAs=Image
     imageType=NoImage
     mirroring=NoMirror
     resourceName=None
     resourceObject=/FortUILibrary/Styles/Border/Materials/M_Border_SoftShadow_DarkBLue
     tiling=NoTile
     imageSize=<32 32>
     margin=<ltrb 0 0 0 0>
     outlineSettings {
      bUseBrushTransparency=false
      roundingType=HalfHeightRadius
      width=0
      color {colorUseRule=UseColor_Specified, specifiedColor=<rgba 0 0 0 0>}
      cornerRadii=<0 0 0 0>
     }
     tintColor {colorUseRule=UseColor_Specified, specifiedColor=<rgba 1 1 1 0.7>}
     uVRegion {bIsValid=false, max=<0 0>, min=<0 0>}
    }
    padding=<ltrb 0 0 0 0>
   }
  }
  SizeBox_1 {
   widgetClassPath=&SizeBox
   parent="[Border[ Dialog Content"
   widget=SizeBox_1
   properties {bOverride_WidthOverride=true, maxDesiredWidth=1075, minDesiredHeight=600, minDesiredWidth=1000, widthOverride=1075}
   slot {class=BorderSlot, properties {horizontalAlignment=HAlign_Center, verticalAlignment=VAlign_Center, padding=<ltrb 0 0 0 0>}}
  }
  VerticalBox_354 {widgetClassPath=&VerticalBox, parent=SizeBox_1, widget=VerticalBox_354, slot {class=SizeBoxSlot}}
  WidgetSwitcher_SeverityImage {
   widgetClassPath=&WidgetSwitcher
   bInherited=true
   bIsVariable=true
   parent=VerticalBox_354
   widget=WidgetSwitcher_SeverityImage
   slot {class=VerticalBoxSlot, properties {horizontalAlignment=HAlign_Center, verticalAlignment=VAlign_Center}}
  }
  Image_Mild {
   widgetClassPath=&Image
   bInherited=true
   bIsVariable=true
   parent=WidgetSwitcher_SeverityImage
   widget=Image_Mild
   properties {
    brush {
     bIsDynamicallyLoaded=false
     drawAs=Image
     imageType=NoImage
     mirroring=NoMirror
     resourceName=None
     resourceObject=/Game/UI/Foundation/Shell/Textures/T-Icon-Warning-128
     tiling=NoTile
     imageSize=<90 90>
     margin=<ltrb 0 0 0 0>
     outlineSettings {
      bUseBrushTransparency=false
      roundingType=HalfHeightRadius
      width=0
      color {colorUseRule=UseColor_Specified, specifiedColor=<rgba 0 0 0 0>}
      cornerRadii=<0 0 0 0>
     }
     tintColor {colorUseRule=UseColor_Specified, specifiedColor=<rgba 1 1 1 1>}
     uVRegion {bIsValid=false, max=<0 0>, min=<0 0>}
    }
    colorAndOpacity=<rgba 1 0.8962 0 1>
   }
   slot {class=WidgetSwitcherSlot}
  }
  Image_Severe {
   widgetClassPath=&Image
   bInherited=true
   bIsVariable=true
   parent=WidgetSwitcher_SeverityImage
   widget=Image_Severe
   properties {
    brush {
     bIsDynamicallyLoaded=false
     drawAs=Image
     imageType=NoImage
     mirroring=NoMirror
     resourceName=None
     resourceObject=/Game/UI/Foundation/Shell/Textures/T-Icon-HardError-64
     tiling=NoTile
     imageSize=<64 64>
     margin=<ltrb 0 0 0 0>
     outlineSettings {
      bUseBrushTransparency=false
      roundingType=HalfHeightRadius
      width=0
      color {colorUseRule=UseColor_Specified, specifiedColor=<rgba 0 0 0 0>}
      cornerRadii=<0 0 0 0>
     }
     tintColor {colorUseRule=UseColor_Specified, specifiedColor=<rgba 1 1 1 1>}
     uVRegion {bIsValid=false, max=<0 0>, min=<0 0>}
    }
    colorAndOpacity=<rgba 1 0.0597 0.0199 1>
   }
   slot {class=WidgetSwitcherSlot}
  }
  VerticalBox_0 {
   widgetClassPath=&VerticalBox
   parent=VerticalBox_354
   widget=VerticalBox_0
   slot {class=VerticalBoxSlot, properties {verticalAlignment=VAlign_Center, size {sizeRule=Fill, value=1}}}
  }
  CommonText_Title {
   widgetClassPath=&CommonTextBlock
   bInherited=true
   bIsVariable=true
   parent=VerticalBox_0
   widget=CommonText_Title
   properties {
    autoWrapText=true
    justification=Center
    lineHeightPercentage=0.9
    text=A PROBLEM OCCURRED
    textTransformPolicy=ToUpper
    font {
     bForceMonospaced=false
     bMaterialIsStencil=false
     fontMaterial=null
     fontObject=&Burbank
     letterSpacing=0
     monospacedWidth=1
     size=48
     skewAmount=0
     typefaceFontName=Burbank-Black
     outlineSettings {
      bApplyOutlineToDropShadows=false
      bMiteredCorners=false
      bSeparateFillAlpha=false
      outlineMaterial=null
      outlineSize=0
      outlineColor=<rgba 0 0 0 1>
     }
    }
    margin=<ltrb 0 0 0 12>
    renderTransform {angle=0, scale=<1 1>, shear=<-12 0>, translation=<0 0>}
    shadowOffset=<0 0>
   }
   slot {class=VerticalBoxSlot}
  }
  Overlay_2 {
   widgetClassPath=&Overlay
   parent=VerticalBox_0
   widget=Overlay_2
   slot {class=VerticalBoxSlot, properties {size {sizeRule=Fill, value=1}}}
  }
  SizeBox_3 {
   widgetClassPath=&SizeBox
   parent=Overlay_2
   widget=SizeBox_3
   properties {bOverride_MaxDesiredHeight=true, clipping=ClipToBounds, maxDesiredHeight=450}
   slot {class=OverlaySlot, properties {horizontalAlignment=HAlign_Fill, padding=<ltrb 0 8 0 8>}}
  }
  ScrollBox_MainError {
   widgetClassPath=&ScrollBox
   bInherited=true
   bIsVariable=true
   parent=SizeBox_3
   widget=ScrollBox_MainError
   properties {
    clipping=Inherit
    widgetBarStyle {
     thickness=16
     draggedThumbImage {
      bIsDynamicallyLoaded=false
      drawAs=Box
      imageType=FullColor
      mirroring=NoMirror
      resourceName=../../../Engine/Content/Slate/Common/Scrollbar_Thumb.png
      resourceObject=/Game/Athena/UI/ReplayBrowser/M_Ui_Skewed_Vfor9slices_BG_Separator_Inst
      tiling=NoTile
      imageSize=<0 0>
      margin=<ltrb 0.25 0.25 0.25 0.25>
      outlineSettings {
       bUseBrushTransparency=false
       roundingType=HalfHeightRadius
       width=0
       color {colorUseRule=UseColor_Specified, specifiedColor=<rgba 0 0 0 0>}
       cornerRadii=<0 0 0 0>
      }
      tintColor {colorUseRule=UseColor_Specified, specifiedColor=<rgba 0.6038 0.8879 1 1>}
      uVRegion {bIsValid=false, max=<0 0>, min=<0 0>}
     }
     horizontalBackgroundImage {
      bIsDynamicallyLoaded=false
      drawAs=NoDrawType
      imageType=NoImage
      mirroring=NoMirror
      resourceName=None
      resourceObject=null
      tiling=NoTile
      imageSize=<0 0>
      margin=<ltrb 0 0 0 0>
      outlineSettings {
       bUseBrushTransparency=false
       roundingType=HalfHeightRadius
       width=0
       color {colorUseRule=UseColor_Specified, specifiedColor=<rgba 0 0 0 0>}
       cornerRadii=<0 0 0 0>
      }
      tintColor {colorUseRule=UseColor_Specified, specifiedColor=<rgba 1 1 1 1>}
      uVRegion {bIsValid=false, max=<0 0>, min=<0 0>}
     }
     horizontalBottomSlotImage {
      bIsDynamicallyLoaded=false
      drawAs=NoDrawType
      imageType=FullColor
      mirroring=NoMirror
      resourceName=../../../Engine/Content/Slate/Common/Scrollbar_Background_Horizontal.png
      resourceObject=null
      tiling=NoTile
      imageSize=<8 8>
      margin=<ltrb 0 0 0 0>
      outlineSettings {
       bUseBrushTransparency=false
       roundingType=HalfHeightRadius
       width=0
       color {colorUseRule=UseColor_Specified, specifiedColor=<rgba 0 0 0 0>}
       cornerRadii=<0 0 0 0>
      }
      tintColor {colorUseRule=UseColor_Specified, specifiedColor=<rgba 1 1 1 1>}
      uVRegion {bIsValid=false, max=<0 0>, min=<0 0>}
     }
     horizontalTopSlotImage {
      bIsDynamicallyLoaded=false
      drawAs=NoDrawType
      imageType=FullColor
      mirroring=NoMirror
      resourceName=../../../Engine/Content/Slate/Common/Scrollbar_Background_Horizontal.png
      resourceObject=null
      tiling=NoTile
      imageSize=<8 8>
      margin=<ltrb 0 0 0 0>
      outlineSettings {
       bUseBrushTransparency=false
       roundingType=HalfHeightRadius
       width=0
       color {colorUseRule=UseColor_Specified, specifiedColor=<rgba 0 0 0 0>}
       cornerRadii=<0 0 0 0>
      }
      tintColor {colorUseRule=UseColor_Specified, specifiedColor=<rgba 1 1 1 1>}
      uVRegion {bIsValid=false, max=<0 0>, min=<0 0>}
     }
     hoveredThumbImage {
      bIsDynamicallyLoaded=false
      drawAs=Box
      imageType=FullColor
      mirroring=NoMirror
      resourceName=../../../Engine/Content/Slate/Common/Scrollbar_Thumb.png
      resourceObject=&M_Ui_Skewed_Vfor9slices_BG_Separator_Inst
      tiling=NoTile
      imageSize=<0 0>
      margin=<ltrb 0.25 0.25 0.25 0.25>
      outlineSettings {
       bUseBrushTransparency=false
       roundingType=HalfHeightRadius
       width=0
       color {colorUseRule=UseColor_Specified, specifiedColor=<rgba 0 0 0 0>}
       cornerRadii=<0 0 0 0>
      }
      tintColor {colorUseRule=UseColor_Specified, specifiedColor=<rgba 0.2159 0.6584 1 1>}
      uVRegion {bIsValid=false, max=<0 0>, min=<0 0>}
     }
     normalThumbImage {
      bIsDynamicallyLoaded=false
      drawAs=Box
      imageType=FullColor
      mirroring=NoMirror
      resourceName=../../../Engine/Content/Slate/Common/Scrollbar_Thumb.png
      resourceObject=&M_Ui_Skewed_Vfor9slices_BG_Separator_Inst
      tiling=NoTile
      imageSize=<0 0>
      margin=<ltrb 0.25 0.25 0.25 0.25>
      outlineSettings {
       bUseBrushTransparency=false
       roundingType=HalfHeightRadius
       width=0
       color {colorUseRule=UseColor_Specified, specifiedColor=<rgba 0 0 0 0>}
       cornerRadii=<0 0 0 0>
      }
      tintColor {colorUseRule=UseColor_Specified, specifiedColor=<rgba 0 0.0908 1 1>}
      uVRegion {bIsValid=false, max=<0 0>, min=<0 0>}
     }
     verticalBackgroundImage {
      bIsDynamicallyLoaded=false
      drawAs=Image
      imageType=NoImage
      mirroring=NoMirror
      resourceName=None
      resourceObject=&M_Ui_Skewed_Vfor9slices_BG_Separator_Inst
      tiling=NoTile
      imageSize=<0 0>
      margin=<ltrb 0 0 0 0>
      outlineSettings {
       bUseBrushTransparency=false
       roundingType=HalfHeightRadius
       width=0
       color {colorUseRule=UseColor_Specified, specifiedColor=<rgba 0 0 0 0>}
       cornerRadii=<0 0 0 0>
      }
      tintColor {colorUseRule=UseColor_Specified, specifiedColor=<rgba 0 0 0 0.25>}
      uVRegion {bIsValid=false, max=<0 0>, min=<0 0>}
     }
     verticalBottomSlotImage {
      bIsDynamicallyLoaded=false
      drawAs=NoDrawType
      imageType=FullColor
      mirroring=NoMirror
      resourceName=../../../Engine/Content/Slate/Common/Scrollbar_Background_Vertical.png
      resourceObject=null
      tiling=NoTile
      imageSize=<8 8>
      margin=<ltrb 0 0 0 0>
      outlineSettings {
       bUseBrushTransparency=false
       roundingType=HalfHeightRadius
       width=0
       color {colorUseRule=UseColor_Specified, specifiedColor=<rgba 0 0 0 0>}
       cornerRadii=<0 0 0 0>
      }
      tintColor {colorUseRule=UseColor_Specified, specifiedColor=<rgba 1 1 1 1>}
      uVRegion {bIsValid=false, max=<0 0>, min=<0 0>}
     }
     verticalTopSlotImage {
      bIsDynamicallyLoaded=false
      drawAs=NoDrawType
      imageType=FullColor
      mirroring=NoMirror
      resourceName=../../../Engine/Content/Slate/Common/Scrollbar_Background_Vertical.png
      resourceObject=null
      tiling=NoTile
      imageSize=<8 8>
      margin=<ltrb 0 0 0 0>
      outlineSettings {
       bUseBrushTransparency=false
       roundingType=HalfHeightRadius
       width=0
       color {colorUseRule=UseColor_Specified, specifiedColor=<rgba 0 0 0 0>}
       cornerRadii=<0 0 0 0>
      }
      tintColor {colorUseRule=UseColor_Specified, specifiedColor=<rgba 1 1 1 1>}
      uVRegion {bIsValid=false, max=<0 0>, min=<0 0>}
     }
    }
    widgetStyle {
     barThickness=9
     bottomShadowBrush {
      bIsDynamicallyLoaded=false
      drawAs=NoDrawType
      imageType=FullColor
      mirroring=NoMirror
      resourceName=../../../Engine/Content/Slate/Common/ScrollBoxShadowBottom.png
      resourceObject=null
      tiling=NoTile
      imageSize=<16 8>
      margin=<ltrb 0.5 0 0.5 1>
      outlineSettings {
       bUseBrushTransparency=false
       roundingType=HalfHeightRadius
       width=0
       color {colorUseRule=UseColor_Specified, specifiedColor=<rgba 0 0 0 0>}
       cornerRadii=<0 0 0 0>
      }
      tintColor {colorUseRule=UseColor_Specified, specifiedColor=<rgba 1 1 1 1>}
      uVRegion {bIsValid=false, max=<0 0>, min=<0 0>}
     }
     horizontalScrolledContentPadding=<ltrb 0 0 1 0>
     leftShadowBrush {
      bIsDynamicallyLoaded=false
      drawAs=Box
      imageType=FullColor
      mirroring=NoMirror
      resourceName=../../../Engine/Content/Slate/Common/ScrollBoxShadowLeft.png
      resourceObject=null
      tiling=NoTile
      imageSize=<8 16>
      margin=<ltrb 1 0.5 0 0.5>
      outlineSettings {
       bUseBrushTransparency=false
       roundingType=HalfHeightRadius
       width=0
       color {colorUseRule=UseColor_Specified, specifiedColor=<rgba 0 0 0 0>}
       cornerRadii=<0 0 0 0>
      }
      tintColor {colorUseRule=UseColor_Specified, specifiedColor=<rgba 1 1 1 1>}
      uVRegion {bIsValid=false, max=<0 0>, min=<0 0>}
     }
     rightShadowBrush {
      bIsDynamicallyLoaded=false
      drawAs=Box
      imageType=FullColor
      mirroring=NoMirror
      resourceName=../../../Engine/Content/Slate/Common/ScrollBoxShadowRight.png
      resourceObject=null
      tiling=NoTile
      imageSize=<8 16>
      margin=<ltrb 0 0.5 1 0.5>
      outlineSettings {
       bUseBrushTransparency=false
       roundingType=HalfHeightRadius
       width=0
       color {colorUseRule=UseColor_Specified, specifiedColor=<rgba 0 0 0 0>}
       cornerRadii=<0 0 0 0>
      }
      tintColor {colorUseRule=UseColor_Specified, specifiedColor=<rgba 1 1 1 1>}
      uVRegion {bIsValid=false, max=<0 0>, min=<0 0>}
     }
     topShadowBrush {
      bIsDynamicallyLoaded=false
      drawAs=NoDrawType
      imageType=FullColor
      mirroring=NoMirror
      resourceName=../../../Engine/Content/Slate/Common/ScrollBoxShadowTop.png
      resourceObject=null
      tiling=NoTile
      imageSize=<16 8>
      margin=<ltrb 0.5 1 0.5 0>
      outlineSettings {
       bUseBrushTransparency=false
       roundingType=HalfHeightRadius
       width=0
       color {colorUseRule=UseColor_Specified, specifiedColor=<rgba 0 0 0 0>}
       cornerRadii=<0 0 0 0>
      }
      tintColor {colorUseRule=UseColor_Specified, specifiedColor=<rgba 1 1 1 1>}
      uVRegion {bIsValid=false, max=<0 0>, min=<0 0>}
     }
     verticalScrolledContentPadding=<ltrb 0 0 0 1>
    }
   }
   slot {class=SizeBoxSlot, properties {padding=<ltrb 12 24 12 24>}}
  }
  CommonText_MainError {
   widgetClassPath=&CommonTextBlock
   bInherited=true
   bIsVariable=true
   parent=ScrollBox_MainError
   widget=CommonText_MainError
   properties {
    autoWrapText=true
    justification=Center
    text="Lorem ipsum dolor sit amet, consectetur adipiscing elit, sed do eiusmod tempor incididunt ut labore et dolore magna aliqua. Ut enim ad minim veniam, quis nostrud exercitation ullamco laboris nisi ut aliquip ex ea commodo consequat.\r\nLorem ipsum dolor sit amet, consectetur adipiscing elit, sed do eiusmod tempor incididunt ut labore et dolore magna aliqua. Ut enim ad minim veniam, quis nostrud exercitation ullamco laboris nisi ut aliquip ex ea commodo consequat.Lorem ipsum dolor sit amet, consectetur adipiscing elit, sed do eiusmod tempor incididunt ut labore et dolore magna aliqua. Ut enim ad minim veniam, quis nostrud exercitation ullamco laboris nisi ut aliquip ex ea commodo consequat. \r\nLorem ipsum dolor sit amet, consectetur adipiscing elit, sed do eiusmod tempor incididunt ut labore et dolore magna aliqua. Ut enim ad minim veniam, quis nostrud exercitation ullamco laboris nisi ut aliquip ex ea commodo consequat.Lorem ipsum dolor sit amet, consectetur adipiscing elit, sed do eiusmod tempor incididunt ut labore et dolore magna aliqua. Ut enim ad minim veniam, quis nostrud exercitation ullamco laboris nisi ut aliquip ex ea commodo consequa"
    colorAndOpacity {colorUseRule=UseColor_Specified, specifiedColor=<rgba 0.2159 0.6584 1 1>}
    font {
     bForceMonospaced=false
     bMaterialIsStencil=false
     fontMaterial=null
     fontObject=&BurbankSmall
     letterSpacing=0
     monospacedWidth=1
     size=22
     skewAmount=0
     typefaceFontName=Bold
     outlineSettings {
      bApplyOutlineToDropShadows=false
      bMiteredCorners=false
      bSeparateFillAlpha=false
      outlineMaterial=null
      outlineSize=0
      outlineColor=<rgba 0 0 0 1>
     }
    }
    shadowOffset=<0 0>
   }
   slot {class=ScrollBoxSlot, properties {padding=<ltrb 12 0 12 0>}}
  }
  Image_Separator_Up {
   widgetClassPath=&Image
   bIsVariable=true
   parent=Overlay_2
   widget=Image_Separator_Up
   properties {
    brush {
     bIsDynamicallyLoaded=false
     drawAs=Box
     imageType=NoImage
     mirroring=NoMirror
     resourceName=None
     resourceObject=/Game/Athena/UI/ReplayBrowser/M_Ui_Skewed_Vfor9slices_Inst1
     tiling=NoTile
     imageSize=<16 16>
     margin=<ltrb 0.5 0.5 0.5 0.5>
     outlineSettings {
      bUseBrushTransparency=false
      roundingType=HalfHeightRadius
      width=0
      color {colorUseRule=UseColor_Specified, specifiedColor=<rgba 0 0 0 0>}
      cornerRadii=<0 0 0 0>
     }
     tintColor {colorUseRule=UseColor_Specified, specifiedColor=<rgba 1 1 1 1>}
     uVRegion {bIsValid=false, max=<0 0>, min=<0 0>}
    }
    colorAndOpacity=<rgba 0 0.0908 1 1>
   }
   slot {class=OverlaySlot, properties {horizontalAlignment=HAlign_Fill}}
  }
  Image_Separator_Down {
   widgetClassPath=&Image
   bIsVariable=true
   parent=Overlay_2
   widget=Image_Separator_Down
   properties {
    brush {
     bIsDynamicallyLoaded=false
     drawAs=Box
     imageType=NoImage
     mirroring=NoMirror
     resourceName=None
     resourceObject=&M_Ui_Skewed_Vfor9slices_Inst1
     tiling=NoTile
     imageSize=<16 16>
     margin=<ltrb 0.5 0.5 0.5 0.5>
     outlineSettings {
      bUseBrushTransparency=false
      roundingType=HalfHeightRadius
      width=0
      color {colorUseRule=UseColor_Specified, specifiedColor=<rgba 0 0 0 0>}
      cornerRadii=<0 0 0 0>
     }
     tintColor {colorUseRule=UseColor_Specified, specifiedColor=<rgba 1 1 1 1>}
     uVRegion {bIsValid=false, max=<0 0>, min=<0 0>}
    }
    colorAndOpacity=<rgba 0 0.0908 1 1>
    renderTransform {angle=0, scale=<-1 1>, shear=<0 0>, translation=<0 0>}
   }
   slot {class=OverlaySlot, properties {horizontalAlignment=HAlign_Fill, verticalAlignment=VAlign_Bottom}}
  }
  CommonButton_Close {
   widgetClassPath=&IconTextButton
   bInherited=true
   bIsVariable=true
   parent=VerticalBox_0
   widget=CommonButton_Close
   properties {
    bHideInputActionWithKeyboard=true
    bIsPersistentBinding=true
    buttonText=Continue
    onButtonBaseClicked=(&self:WidgetTree.CommonButton_Close.HandleOnButtonClicked)
    onButtonBaseDoubleClicked=(&self:WidgetTree.CommonButton_Close.HandleOnButtonDoubleClicked)
    onButtonBaseHovered=(&self:WidgetTree.CommonButton_Close.HandleOnButtonHovered)
    onButtonBaseUnhovered=(&self:WidgetTree.CommonButton_Close.HandleOnButtonUnhovered)
    onSelectedChangedBase=(&self:WidgetTree.CommonButton_Close.HandleOnSelectedChanged)
    style=/FortUILibrary/Foundation/Buttons/ButtonStyles/ButtonStyle_NewFeature_M_Yellow.ButtonStyle_NewFeature_M_Yellow_C
    triggeringInputAction {dataTable=&InputActionDataTable, rowName=Continue}
   }
   slot {class=VerticalBoxSlot, properties {horizontalAlignment=HAlign_Center, verticalAlignment=VAlign_Bottom, padding=<ltrb 20 20 20 20>}}
  }
  TextBlock_MainErrorCode {
   widgetClassPath=&TextBlock
   bInherited=true
   bIsVariable=true
   parent=VerticalBox_0
   widget=TextBlock_MainErrorCode
   properties {
    justification=Center
    text="(errors.com.epicgames.common.processing"
    visibility=SelfHitTestInvisible
    wrapTextAt=800
    colorAndOpacity {colorUseRule=UseColor_Specified, specifiedColor=<rgba 0.2159 0.6584 1 1>}
    font {
     bForceMonospaced=false
     bMaterialIsStencil=false
     fontMaterial=null
     fontObject=&BurbankSmall
     letterSpacing=0
     monospacedWidth=1
     size=14
     skewAmount=0
     typefaceFontName=Bold
     outlineSettings {
      bApplyOutlineToDropShadows=false
      bMiteredCorners=false
      bSeparateFillAlpha=false
      outlineMaterial=null
      outlineSize=0
      outlineColor=<rgba 0 0 0 1>
     }
    }
   }
   slot {class=VerticalBoxSlot, properties {horizontalAlignment=HAlign_Center}}
  }
  TextBlock_OverflowErrorCodes {
   widgetClassPath=&TextBlock
   bInherited=true
   bIsVariable=true
   parent=VerticalBox_0
   widget=TextBlock_OverflowErrorCodes
   properties {
    justification=Center
    text="(errors.com.epicgames.common.processing"
    visibility=SelfHitTestInvisible
    wrapTextAt=800
    colorAndOpacity {colorUseRule=UseColor_Specified, specifiedColor=<rgba 0.2159 0.6584 1 1>}
    font {
     bForceMonospaced=false
     bMaterialIsStencil=false
     fontMaterial=null
     fontObject=&BurbankSmall
     letterSpacing=0
     monospacedWidth=1
     size=14
     skewAmount=0
     typefaceFontName=Bold
     outlineSettings {
      bApplyOutlineToDropShadows=false
      bMiteredCorners=false
      bSeparateFillAlpha=false
      outlineMaterial=null
      outlineSize=0
      outlineColor=<rgba 0 0 0 1>
     }
    }
   }
   slot {class=VerticalBoxSlot, properties {horizontalAlignment=HAlign_Center}}
  }
 ]
}
