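# PIN entry modal, built on FortPinModal, for parental-control prompts such as friend invites, game unlocks and purchases.
# Visible keypad buttons and a hidden duplicate set bound to number input actions pass clicks to the parent class, as do Accept, Backspace and Forgot PIN.
# Parent events drive validity icons and background-material gradient, switch default/cabin layouts, and fill an info panel with game, purchase or text details.
# Deactivation resets header and panel.
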WidgetBlueprint /Game/UI/Foundation/Modals/PasscodeModal/PinModal
graphs [
 CollectPinButtons {nodes [
   CollectPinButtons() {is=title K2Node_FunctionEntry}
   > Return Node!(
    execute
    ReturnValue:Array of Fort Pin Button Object References=Make Array!(
     "[0]":Fort Pin Button Object Reference=Get PasscodeButton_0() -> (PasscodeButton_0:Pin Code Button Object Reference)
     "[1]":Fort Pin Button Object Reference=Get PasscodeButton_1() -> (PasscodeButton_1:Pin Code Button Object Reference)
     "[2]":Fort Pin Button Object Reference=Get PasscodeButton_2() -> (PasscodeButton_2:Pin Code Button Object Reference)
     "[3]":Fort Pin Button Object Reference=Get PasscodeButton_3() -> (PasscodeButton_3:Pin Code Button Object Reference)
     "[4]":Fort Pin Button Object Reference=Get PasscodeButton_4() -> (PasscodeButton_4:Pin Code Button Object Reference)
     "[5]":Fort Pin Button Object Reference=Get PasscodeButton_5() -> (PasscodeButton_5:Pin Code Button Object Reference)
     "[6]":Fort Pin Button Object Reference=Get PasscodeButton_6() -> (PasscodeButton_6:Pin Code Button Object Reference)
     "[7]":Fort Pin Button Object Reference=Get PasscodeButton_7() -> (PasscodeButton_7:Pin Code Button Object Reference)
     "[8]":Fort Pin Button Object Reference=Get PasscodeButton_8() -> (PasscodeButton_8:Pin Code Button Object Reference)
     "[9]":Fort Pin Button Object Reference=Get PasscodeButton_9() -> (PasscodeButton_9:Pin Code Button Object Reference)
     "[10]":Fort Pin Button Object Reference=Get Hidden_PasscodeButton_0() -> (Hidden_PasscodeButton_0:Pin Code Button Object Reference)
     "[11]":Fort Pin Button Object Reference=Get Hidden_PasscodeButton_1() -> (Hidden_PasscodeButton_1:Pin Code Button Object Reference)
     "[12]":Fort Pin Button Object Reference=Get Hidden_PasscodeButton_2() -> (Hidden_PasscodeButton_2:Pin Code Button Object Reference)
     "[13]":Fort Pin Button Object Reference=Get Hidden_PasscodeButton_3() -> (Hidden_PasscodeButton_3:Pin Code Button Object Reference)
     "[14]":Fort Pin Button Object Reference=Get Hidden_PasscodeButton_4() -> (Hidden_PasscodeButton_4:Pin Code Button Object Reference)
     "[15]":Fort Pin Button Object Reference=Get Hidden_PasscodeButton_5() -> (Hidden_PasscodeButton_5:Pin Code Button Object Reference)
     "[16]":Fort Pin Button Object Reference=Get Hidden_PasscodeButton_6() -> (Hidden_PasscodeButton_6:Pin Code Button Object Reference)
     "[17]":Fort Pin Button Object Reference=Get Hidden_PasscodeButton_7() -> (Hidden_PasscodeButton_7:Pin Code Button Object Reference)
     "[18]":Fort Pin Button Object Reference=Get Hidden_PasscodeButton_8() -> (Hidden_PasscodeButton_8:Pin Code Button Object Reference)
     "[19]":Fort Pin Button Object Reference=Get Hidden_PasscodeButton_9() -> (Hidden_PasscodeButton_9:Pin Code Button Object Reference)
    ) -> (Array:Array of Fort Pin Button Object References)
   )
  ]}
 EventGraph {nodes [
   "On Button Clicked (Button_Close_1)"!() -> (OutputDelegate, then, Button:Common Button Legacy Object Reference) {is=title K2Node_ComponentBoundEvent}
   > ClosePinModal(self:Fort Pin Modal Object Reference)
   "On Clicked (Button_AboutRatings)"!() -> (OutputDelegate, then, Button:Common Button Base Object Reference) {is=title K2Node_ComponentBoundEvent}
   > n1 = Get!(execute:Exec) -> (then:Exec, else:Exec, CachedGameActivity:Fort Game Activity Object Reference)
   > OpenRatingInfoURL!(
    execute
    self:Fort Game Activity Object Reference=@n1
    PlayerContext:Local Player Object Reference=GetOwningLocalPlayer!(self:Widget Object Reference) -> (ReturnValue:Local Player Object Reference)
   ) -> (then)
   "On Clicked (Button_ForgotPin)"!() -> (OutputDelegate, then, Button:Common Button Base Object Reference) {is=title K2Node_ComponentBoundEvent}
   > HandleForgotPinButtonClicked(self:Fort Pin Modal Object Reference)
   "On Clicked (Button_Close)"!() -> (OutputDelegate, then, Button:Common Button Base Object Reference) {is=title K2Node_ComponentBoundEvent}
   > HandleCloseButtonClicked(self:Fort Pin Modal Object Reference)
   "On Clicked (Button_Accept)"!() -> (OutputDelegate, then, Button:Common Button Base Object Reference) {is=title K2Node_ComponentBoundEvent}
   > HandleEnterButtonClicked(self:Fort Pin Modal Object Reference)
   "On Clicked (Button_Backspace)"!() -> (OutputDelegate, then, Button:Common Button Base Object Reference) {is=title K2Node_ComponentBoundEvent}
   > HandleBackspaceButtonClicked(self:Fort Pin Modal Object Reference)
   n2 = Event PreConstruct()
   > SetCompactUIStyle() {is=title K2Node_CallFunction}
   > Branch(Condition=@n2.IsDesignTime)
   > SetPinButtonNumbers()
   > Set Visual States!(execute, self, Pin States:EPinModalValidityState Enum=Get Preview Pin State() -> (Preview Pin State:EPinModalValidityState Enum=None)) -> (then) {is=title K2Node_CallFunction}
   > SetVisibility(
    self=Get Text_Results() -> (Text_Results:Common Text Object Reference)
    InVisibility=Select!(
     None:ESlateVisibility Enum=Hidden
     Requesting:ESlateVisibility Enum=Hidden
     Success:ESlateVisibility Enum=Hidden
     Invalid:ESlateVisibility Enum=SelfHitTestInvisible
     Index:EPinModalValidityState Enum=Get Preview Pin State() -> (Preview Pin State=None)
    ) -> (ReturnValue)
   )
   n3 = Event OnSetPinState!() -> (OutputDelegate, then, PinModalState:EPinModalValidityState Enum=None) {is=title K2Node_Event}
   > Set Visual States(Pin States=@n3.PinModalState)
   n4 = Event OnShowCabinModeLayout!() -> (OutputDelegate, then, bIsInCabinMode:bool) {is=title K2Node_Event}
   > SetActiveWidget(
    self:Widget Switcher Object Reference=Get CabinSwitcher() -> (CabinSwitcher:Common Activatable Widget Switcher Object Reference)
    Widget=Select!(
     Option 0:Widget Object Reference=Get DefaultLayout() -> (DefaultLayout:Horizontal Box Object Reference)
     Option 1:Widget Object Reference=Get CabinLayout() -> (CabinLayout:Vertical Box Object Reference)
     Index:bool=@n4.bIsInCabinMode
    ) -> (ReturnValue:Widget Object Reference)
   )
   n5 = Event SetupSqueegeeWidgets!() -> (OutputDelegate, then, GameActivity:Fort Game Activity Object Reference) {is=title K2Node_Event}
   > SetActiveWidget(
    self:Common Visibility Switcher Object Reference=Get InfoSwitcher() -> (InfoSwitcher:Common Visibility Switcher Object Reference)
    Widget=Get SqueegeeSlot_UnlockInfo() -> (SqueegeeSlot_UnlockInfo:Squeegee Injection Slot Object Reference)
   )
   > SetUnlockPlayer!(
    execute
    self:Fort Squeegee Widget Injection Slot Object Reference=Get SqueegeeSlot_UnlockInfo()
    UnlockPlayer:Fort Player Controller Object Reference=GetUnlockPlayer!(self:Fort Pin Modal Object Reference) -> (ReturnValue:Fort Player Controller Object Reference)
   ) -> (then)
   > ShowActivityData!(
    execute
    self:Fort Squeegee Widget Injection Slot Object Reference=Get SqueegeeSlot_UnlockInfo()
    GameActivity:Fort Game Activity Object Reference=@n5.GameActivity
   ) -> (then)
   > UpdateAboutRatingsButtonVisibility!(execute, self, GameActivity:Fort Game Activity Object Reference=@n5.GameActivity) -> (then)
   > SetVisibility(self=Get InfoSwitcher(), InVisibility=SelfHitTestInvisible)
   > n6 = Cast To WBP_UIKit_Button_Quiet!(execute, Object=Get Button_Close() -> (Button_Close:WBP UIKit Button Quiet Object Reference)) -> (then, CastFailed, AsWBP UIKit Button Quiet:WBP UIKit Button Quiet Object Reference) {is=title K2Node_DynamicCast}
   > SetText!(execute, self:Fort CTAButton Object Reference=@n6, ButtonText:"Text (by ref)"=MakeLiteralText(Value="LOCTEXT(\"View Game Details\")")) -> (then)
   Event On Deactivated()
   > SetVisibility(self=Get InfoSwitcher(), InVisibility=Collapsed)
   > SetVisibility(self=Get Button_AboutRatings() -> (Button_AboutRatings:WBP UIKit Button Quiet Object Reference), InVisibility=Collapsed)
   > n7 = Cast To WBP_UIKit_Button_Quiet(Object=Get Button_Close())
   > n8 = SetText(self=@n7, ButtonText=MakeLiteralText(Value="LOCTEXT(\"Close\")"))
   Set Text(execute=[@n8.then, @n7.CastFailed], Text="LOCTEXT(\"Enter Pin\")", self=Get HeaderText() -> (HeaderText:Common Text Object Reference))
   > SetVisibility(self=Get SubheaderText() -> (SubheaderText:Common Text Object Reference), InVisibility=Collapsed)
   > UnloadThumbnails(self:WBP Cosmetic Gate Purchase Info Object Reference=Get PurchaseInfo() -> (PurchaseInfo:WBP Cosmetic Gate Purchase Info Object Reference))
   n9 = Event SetupPurchaseRequest!() -> (OutputDelegate, then, OfferVM:Fort Item Shop Offer VM Object Reference) {is=title K2Node_Event}
   > SetActiveWidget(self=Get InfoSwitcher(), Widget=Get PurchaseInfo())
   > Set FortItemShopOfferVM!(
    execute
    FortItemShopOfferVM:Fort Item Shop Offer VM Object Reference=@n9.OfferVM
    self:WBP Cosmetic Gate Purchase Info Object Reference=Get PurchaseInfo()
   ) -> (then, Output_Get:Fort Item Shop Offer VM Object Reference)
   > SetVisibility(self=Get InfoSwitcher(), InVisibility=SelfHitTestInvisible)
   > Set Text(Text="LOCTEXT(\"Purchase\")", self=Get HeaderText())
   > SetVisibility(self=Get SubheaderText(), InVisibility=SelfHitTestInvisible)
   Event Construct()
   > SetPinButtonClickedEvents()
   n10 = Event SetupInfoTextDescription!() -> (OutputDelegate, then, Title:"Text (by ref)", Description:"Text (by ref)") {is=title K2Node_Event}
   > SetActiveWidget(self=Get InfoSwitcher(), Widget=Get InfoText() -> (InfoText:WBP Pin Modal Info Text Object Reference))
   > SetInfoText!(execute, self:WBP Pin Modal Info Text Object Reference=Get InfoText(), Title:Text=@n10.Title, Description:Text=@n10.Description) -> (then)
   > SetVisibility(self=Get InfoSwitcher(), InVisibility=SelfHitTestInvisible)
  ]}
 Handle Pin Button Clicked Event {nodes [
   n1 = Handle Pin Button Clicked Event!() -> (then, Button:Common Button Legacy Object Reference) {is=title K2Node_FunctionEntry}
   > HandlePinButtonClicked!(execute, self:Fort Pin Modal Object Reference, SelectedButton:Common Button Base Object Reference=@n1, TabIndex:int=0) -> (then)
  ]}
 Set Visual States {nodes [
   n1 = Set Visual States!() -> (then, Pin States:EPinModalValidityState Enum=None) {is=title K2Node_FunctionEntry}
   n2 = Get Waiting() -> (Waiting:Image Object Reference)
   n3 = Get InputPin_StateSwitcher() -> (InputPin_StateSwitcher:Common Visibility Switcher Object Reference)
   > SetActiveWidget(
    self=@n3
    Widget=Select!(
     None:Widget Object Reference=@n2
     Requesting:Widget Object Reference=@n2
     Success:Widget Object Reference=Get Good() -> (Good:Image Object Reference)
     Invalid:Widget Object Reference=Get Bad() -> (Bad:Image Object Reference)
     Index:EPinModalValidityState Enum=@n1
    ) -> (ReturnValue)
   )
   > SetVisibility(
    self=@n3
    InVisibility=Select(
     None:ESlateVisibility Enum=Hidden
     Requesting:ESlateVisibility Enum=SelfHitTestInvisible
     Success:ESlateVisibility Enum=SelfHitTestInvisible
     Invalid:ESlateVisibility Enum=SelfHitTestInvisible
     Index=Get Pin States() -> (Pin States:EPinModalValidityState Enum=None)
    ) -> (ReturnValue:ESlateVisibility Enum)
   )
   > n4 = GetDynamicMaterial(self=Get InputPin_Bg() -> (InputPin_Bg:Image Object Reference))
   > SetScalarParameterValue(
    self=@n4
    ParameterName=Display State Gradient
    Value=Select(None:float=0, Requesting:float=0, Success:float=1, Invalid:float=1, Index=Get Pin States() -> (Pin States=None)) -> (ReturnValue:float)
   )
   > SetVectorParameterValue(
    self=@n4
    ParameterName=Gradient Color
    Value=Select(
     None:Linear Color Structure
     Requesting:Linear Color Structure
     Success:Linear Color Structure=Get SuccessColor() -> (SuccessColor:Linear Color Structure)
     Invalid:Linear Color Structure=Get InvalidColor() -> (InvalidColor:Linear Color Structure)
     Index=Get Pin States() -> (Pin States=None)
    ) -> (ReturnValue:Linear Color Structure)
   )
  ]}
 SetCompactUIStyle {nodes [
   SetCompactUIStyle() {is=title K2Node_FunctionEntry}
   > Branch(Condition=IsUsingCompactUI())
   n1 = Get Text_Results()
   > UpdateWidgetStyleTags!(
    execute
    InWidget:Widget Object Reference=@n1
    InAddedTags:Array of Gameplay Tag Structures=Make Array!("[0]":Gameplay Tag Structure) -> (Array:Array of Gameplay Tag Structures)
    InRemovedTags:Array of Gameplay Tag Structures=Make Array("[0]"=(TagName="StyleSheet.Style.20"))
   ) -> (then, ReturnValue:bool)
   > SetMobileFontSizeMultiplier!(execute, self:Common Text Object Reference=@n1, InMobileFontSizeMultiplier:float=1.14) -> (then)
  ]}
 SetPinButtonClickedEvents {nodes [
   SetPinButtonClickedEvents() {is=title K2Node_FunctionEntry}
   > n1 = For Each Loop(Array:Array of Pin Code Button Object References=Make Array!(
     "[0]":Pin Code Button Object Reference=Get PasscodeButton_0()
     "[1]":Pin Code Button Object Reference=Get PasscodeButton_1()
     "[2]":Pin Code Button Object Reference=Get PasscodeButton_2()
     "[3]":Pin Code Button Object Reference=Get PasscodeButton_3()
     "[4]":Pin Code Button Object Reference=Get PasscodeButton_4()
     "[5]":Pin Code Button Object Reference=Get PasscodeButton_5()
     "[6]":Pin Code Button Object Reference=Get PasscodeButton_6()
     "[7]":Pin Code Button Object Reference=Get PasscodeButton_7()
     "[8]":Pin Code Button Object Reference=Get PasscodeButton_8()
     "[9]":Pin Code Button Object Reference=Get PasscodeButton_9()
    ) -> (Array:Array of Pin Code Button Object References)) -> (Array Element:Pin Code Button Object Reference)
   > Bind Event to On Button Clicked(self:Common Button Legacy Object Reference=@n1, Delegate=Create Event()) {is=title K2Node_AddDelegate}
   n2 = Make Array(
    "[0]"=Get Hidden_PasscodeButton_0()
    "[1]"=Get Hidden_PasscodeButton_1()
    "[2]"=Get Hidden_PasscodeButton_2()
    "[3]"=Get Hidden_PasscodeButton_3()
    "[4]"=Get Hidden_PasscodeButton_4()
    "[5]"=Get Hidden_PasscodeButton_5()
    "[6]"=Get Hidden_PasscodeButton_6()
    "[7]"=Get Hidden_PasscodeButton_7()
    "[8]"=Get Hidden_PasscodeButton_8()
    "[9]"=Get Hidden_PasscodeButton_9()
   )
   > Branch(Condition=Is Valid Index!(TargetArray:Array of Pin Code Button Object References=@n2, IndexToTest:int=@n1.Array Index) -> (ReturnValue:bool))
   > Bind Event to On Button Clicked(
    self="Get (a copy)"(Array:Array of Pin Code Button Object References=@n2, Dimension 1=@n1.Array Index) -> (Output:Pin Code Button Object Reference)
    Delegate=Create Event()
   )
  ]}
 SetPinButtonNumbers {nodes [
   SetPinButtonNumbers() {is=title K2Node_FunctionEntry}
   > n1 = For Each Loop(Array=Make Array(
     "[0]"=Get PasscodeButton_0()
     "[1]"=Get PasscodeButton_1()
     "[2]"=Get PasscodeButton_2()
     "[3]"=Get PasscodeButton_3()
     "[4]"=Get PasscodeButton_4()
     "[5]"=Get PasscodeButton_5()
     "[6]"=Get PasscodeButton_6()
     "[7]"=Get PasscodeButton_7()
     "[8]"=Get PasscodeButton_8()
     "[9]"=Get PasscodeButton_9()
    ))
   > Set Pin Number!(execute, self:Pin Code Button Object Reference=@n1, Pin:String=Get PinValue!(self:Fort Pin Button Object Reference=@n1) -> (PinValue:String)) -> (then) {is=title K2Node_CallFunction}
  ]}
 UpdateAboutRatingsButtonVisibility {nodes [
   n1 = UpdateAboutRatingsButtonVisibility!() -> (then, GameActivity:Fort Game Activity Object Reference) {is=title K2Node_FunctionEntry}
   > n2 = Set CachedGameActivity(CachedGameActivity:Fort Game Activity Object Reference=@n1) -> (Output_Get:Fort Game Activity Object Reference)
   > n3 = Is Valid(InputObject=@n2)
   > n4 = Branch(Condition=HasAgeRating!(self:Fort Game Activity Object Reference=@n2, PlayerContext:Local Player Object Reference=GetOwningLocalPlayer()) -> (ReturnValue:bool))
   > SetVisibility(self=Get Button_AboutRatings(), InVisibility=Visible)
   SetVisibility(execute=[@n4.else, @n3.Is Not Valid], self=Get Button_AboutRatings(), InVisibility=Collapsed)
  ]}
]
variables [
 CachedGameActivity {type=FortGameActivity, default=null}
 InvalidColor {type=LinearColor, default=<rgba 1 0.147 0.1329 1>}
 Preview Pin State {type=EPinModalValidityState, default=Invalid}
 RequestingColor {type=LinearColor, default=<rgba 0.0331 0.402 1 1>}
 SuccessColor {type=LinearColor, default=<rgba 0.0331 0.9823 0.2232 1>}
]
widgets {
 info {inheritedWidgetCount=6, parentClass=/Script/FortniteUI.FortPinModal, rootWidgetClass=&Overlay, widgetCount=63}
 widgets [
  Overlay_20 {widgetClassPath=&Overlay, widget=Overlay_20}
  WBP_UIKit_Scrim {
   widgetClassPath=/FortUILibrary/Systems/Dialog/WBP_UIKit_Scrim
   bIsVariable=true
   parent=Overlay_20
   widget=WBP_UIKit_Scrim
   slot {class=OverlaySlot, properties {horizontalAlignment=HAlign_Fill, verticalAlignment=VAlign_Fill}}
  }
  SizeBox_84 {
   widgetClassPath=&SizeBox
   parent=Overlay_20
   widget=SizeBox_84
   properties {
    bOverride_MaxDesiredHeight=true
    bOverride_WidthOverride=true
    heightOverride=896
    maxDesiredHeight=896
    maxDesiredWidth=1752
    minDesiredHeight=900
    minDesiredWidth=1424
    widthOverride=1456
   }
   slot {class=OverlaySlot, properties {horizontalAlignment=HAlign_Center, verticalAlignment=VAlign_Center}}
  }
  Overlay_0 {widgetClassPath=&Overlay, parent=SizeBox_84, widget=Overlay_0, slot {class=SizeBoxSlot}}
  WBP_CaptureForPostBufferUpdate {
   widgetClassPath=&WBP_CaptureForPostBufferUpdate
   bIsVariable=true
   parent=Overlay_0
   slot {class=OverlaySlot, properties {horizontalAlignment=HAlign_Fill, verticalAlignment=VAlign_Fill}}
  }
  WBP_UIKit_Backplate {
   widgetClassPath=&WBP_UIKit_Backplate
   bIsVariable=true
   parent=Overlay_0
   widget=WBP_UIKit_Backplate
   properties {brightness=Dark}
   slot {class=OverlaySlot, properties {horizontalAlignment=HAlign_Fill, verticalAlignment=VAlign_Fill}}
  }
  CabinSwitcher {
   widgetClassPath=&CommonActivatableWidgetSwitcher
   bIsVariable=true
   parent=Overlay_0
   widget=CabinSwitcher
   properties {transitionDuration=0}
   slot {class=OverlaySlot, properties {horizontalAlignment=HAlign_Fill, verticalAlignment=VAlign_Fill}}
  }
  DefaultLayout {widgetClassPath=&HorizontalBox, bIsVariable=true, parent=CabinSwitcher, widget=DefaultLayout, slot {class=WidgetSwitcherSlot}}
  LeftSide_Panel {
   widgetClassPath=&VerticalBox
   parent=DefaultLayout
   widget=LeftSide_Panel
   slot {class=HorizontalBoxSlot, properties {padding=<ltrb 80 64 40 64>, size {sizeRule=Fill, value=1}}}
  }
  HeaderText {
   widgetClassPath=&CommonTextBlock
   bIsVariable=true
   parent=LeftSide_Panel
   widget=HeaderText
   properties {
    applyLineHeightToBottomLine=false
    autoWrapText=true
    text=ENTER PIN
    font {
     bForceMonospaced=false
     bMaterialIsStencil=false
     fontMaterial=null
     fontObject=&Burbank
     letterSpacing=50
     monospacedWidth=1
     size=54
     skewAmount=0
     typefaceFontName=BurbankRegular-Black
     outlineSettings {
      bApplyOutlineToDropShadows=false
      bMiteredCorners=false
      bSeparateFillAlpha=false
      outlineMaterial=null
      outlineSize=0
      outlineColor=<rgba 0 0 0 1>
     }
    }
    shadowOffset=<0 0>
   }
   slot {class=VerticalBoxSlot, properties {verticalAlignment=VAlign_Top}}
  }
  SubheaderText {
   widgetClassPath=&CommonTextBlock
   bIsVariable=true
   parent=LeftSide_Panel
   widget=SubheaderText
   properties {
    autoWrapText=true
    text=ENTER PIN
    visibility=Collapsed
    font {
     bForceMonospaced=false
     bMaterialIsStencil=false
     fontMaterial=null
     fontObject=&Burbank
     letterSpacing=50
     monospacedWidth=1
     size=36
     skewAmount=0
     typefaceFontName=Burbank-PlayerPower
     outlineSettings {
      bApplyOutlineToDropShadows=false
      bMiteredCorners=false
      bSeparateFillAlpha=false
      outlineMaterial=null
      outlineSize=0
      outlineColor=<rgba 0 0 0 1>
     }
    }
    shadowOffset=<0 0>
   }
   slot {class=VerticalBoxSlot, properties {verticalAlignment=VAlign_Top, padding=<ltrb 0 8 0 0>}}
  }
  Text_Description {
   widgetClassPath=&CommonRichTextBlock
   bInherited=true
   bIsVariable=true
   parent=LeftSide_Panel
   widget=Text_Description
   properties {
    applyLineHeightToBottomLine=false
    autoWrapText=true
    text="Enter your parental control PIN to accept friend invite from <playername>TheTacoTruck</>"
    textStyleSet=/Game/UI/Foundation/Modals/PasscodeModal/Styles/UI_Model_Passcode_RichTextStyles
    defaultTextStyleOverride {
     overflowPolicy=Clip
     transformPolicy=None
     colorAndOpacity {colorUseRule=UseColor_Specified, specifiedColor=<rgba 1 1 1 1>}
     font {
      bForceMonospaced=false
      bMaterialIsStencil=false
      fontMaterial=null
      fontObject=&NotoSans
      letterSpacing=0
      monospacedWidth=1
      size=24
      skewAmount=0
      typefaceFontName=NotoSans-Regular
      outlineSettings {
       bApplyOutlineToDropShadows=false
       bMiteredCorners=false
       bSeparateFillAlpha=false
       outlineMaterial=null
       outlineSize=0
       outlineColor=<rgba 0 0 0 1>
      }
     }
     highlightColor {colorUseRule=UseColor_Specified, specifiedColor=<rgba 0 0 0 1>}
     highlightShape {
      bIsDynamicallyLoaded=false
      drawAs=Image
      imageType=NoImage
      mirroring=NoMirror
      resourceName=None
      resourceObject=null
      tiling=NoTile
      imageSize=<32 32>
      margin=<ltrb 0 0 0 0>
      outlineSettings {
       bUseBrushTransparency=false
       roundingType=HalfHeightRadius
       width=0
       color {colorUseRule=UseColor_Specified, specifiedColor=<rgba 0 0 0 0>}
       cornerRadii=<0 0 0 0>
      }
      tintColor {colorUseRule=UseColor_Specified, specifiedColor=<rgba 1 1 1 1>}
      uVRegion {bIsValid=false, max=<0 0>, min=<0 0>}
     }
     selectedBackgroundColor {colorUseRule=UseColor_Foreground, specifiedColor=<rgba 1 0 1 1>}
     shadowColorAndOpacity=<rgba 0 0 0 1>
     shadowOffset=<0 0>
     strikeBrush {
      bIsDynamicallyLoaded=false
      drawAs=Image
      imageType=NoImage
      mirroring=NoMirror
      resourceName=None
      resourceObject=null
      tiling=NoTile
      imageSize=<32 32>
      margin=<ltrb 0 0 0 0>
      outlineSettings {
       bUseBrushTransparency=false
       roundingType=HalfHeightRadius
       width=0
       color {colorUseRule=UseColor_Specified, specifiedColor=<rgba 0 0 0 0>}
       cornerRadii=<0 0 0 0>
      }
      tintColor {colorUseRule=UseColor_Specified, specifiedColor=<rgba 1 1 1 1>}
      uVRegion {bIsValid=false, max=<0 0>, min=<0 0>}
     }
     underlineBrush {
      bIsDynamicallyLoaded=false
      drawAs=Image
      imageType=NoImage
      mirroring=NoMirror
      resourceName=None
      resourceObject=null
      tiling=NoTile
      imageSize=<32 32>
      margin=<ltrb 0 0 0 0>
      outlineSettings {
       bUseBrushTransparency=false
       roundingType=HalfHeightRadius
       width=0
       color {colorUseRule=UseColor_Specified, specifiedColor=<rgba 0 0 0 0>}
       cornerRadii=<0 0 0 0>
      }
      tintColor {colorUseRule=UseColor_Specified, specifiedColor=<rgba 1 1 1 1>}
      uVRegion {bIsValid=false, max=<0 0>, min=<0 0>}
     }
    }
   }
   slot {class=VerticalBoxSlot, properties {verticalAlignment=VAlign_Top, padding=<ltrb 0 16 0 0>}}
  }
  InputPin_Container {
   widgetClassPath=&Overlay
   parent=LeftSide_Panel
   widget=InputPin_Container
   slot {class=VerticalBoxSlot, properties {horizontalAlignment=HAlign_Left, verticalAlignment=VAlign_Top, padding=<ltrb 0 24 0 0>}}
  }
  InputPin_MinHeight {
   widgetClassPath=&Spacer
   parent=InputPin_Container
   widget=InputPin_MinHeight
   properties {size=<480 88>}
   slot {class=OverlaySlot, properties {horizontalAlignment=HAlign_Fill, verticalAlignment=VAlign_Fill}}
  }
  InputPin_Bg {
   widgetClassPath=&Image
   bIsVariable=true
   parent=InputPin_Container
   widget=InputPin_Bg
   properties {
    visibility=SelfHitTestInvisible
    brush {
     bIsDynamicallyLoaded=false
     drawAs=Image
     imageType=FullColor
     mirroring=NoMirror
     resourceName=None
     resourceObject=/Game/UI/Foundation/Modals/PasscodeModal/Materials/MI_UI_InputPin_Bg
     tiling=NoTile
     imageSize=<32 32>
     margin=<ltrb 0 0 0 0>
     outlineSettings {
      bUseBrushTransparency=false
      roundingType=HalfHeightRadius
      width=0
      color {colorUseRule=UseColor_Specified, specifiedColor=<rgba 0 0 0 0>}
      cornerRadii=<0 0 0 0>
     }
     tintColor {colorUseRule=UseColor_Specified, specifiedColor=<rgba 1 1 1 1>}
     uVRegion {bIsValid=false, max=<0 0>, min=<0 0>}
    }
   }
   slot {class=OverlaySlot, properties {horizontalAlignment=HAlign_Fill, verticalAlignment=VAlign_Fill}}
  }
  DynamicEntryBox_Chars {
   widgetClassPath=&DynamicEntryBox
   bInherited=true
   bIsVariable=true
   parent=InputPin_Container
   widget=DynamicEntryBox_Chars
   properties {
    entryVerticalAlignment=VAlign_Center
    entryWidgetClass=&/PinDot.PinDot_C
    numDesignerPreviewEntries=6
    entrySpacing=<24 0>
   }
   slot {class=OverlaySlot, properties {verticalAlignment=VAlign_Center, padding=<ltrb 32 24 32 24>}}
  }
  InputPin_StateSwitcher {
   widgetClassPath=&CommonVisibilitySwitcher
   bIsVariable=true
   parent=InputPin_Container
   widget=InputPin_StateSwitcher
   properties {visibility=HitTestInvisible}
   slot {class=OverlaySlot, properties {horizontalAlignment=HAlign_Right, verticalAlignment=VAlign_Center, padding=<ltrb 0 0 16 0>}}
  }
  Waiting {
   widgetClassPath=&Image
   bIsVariable=true
   parent=InputPin_StateSwitcher
   widget=Waiting
   properties {
    visibility=SelfHitTestInvisible
    brush {
     bIsDynamicallyLoaded=false
     drawAs=Image
     imageType=NoImage
     mirroring=NoMirror
     resourceName=None
     resourceObject=/FortUILibrary/UIKit/Throbber/MI_UIKit_Monochrome_Throbber
     tiling=NoTile
     imageSize=<56 56>
     margin=<ltrb 0 0 0 0>
     outlineSettings {
      bUseBrushTransparency=false
      roundingType=HalfHeightRadius
      width=0
      color {colorUseRule=UseColor_Specified, specifiedColor=<rgba 0 0 0 0>}
      cornerRadii=<0 0 0 0>
     }
     tintColor {colorUseRule=UseColor_Specified, specifiedColor=<rgba 1 1 1 1>}
     uVRegion {bIsValid=false, max=<0 0>, min=<0 0>}
    }
   }
   slot {class=CommonVisibilitySwitcherSlot}
  }
  Good {
   widgetClassPath=&Image
   bIsVariable=true
   parent=InputPin_StateSwitcher
   widget=Good
   properties {
    visibility=SelfHitTestInvisible
    brush {
     bIsDynamicallyLoaded=false
     drawAs=Image
     imageType=FullColor
     mirroring=NoMirror
     resourceName=None
     resourceObject=/FortUILibrary/Textures/MaterialInstances/MI_UIKit_Icon_Check
     tiling=NoTile
     imageSize=<56 56>
     margin=<ltrb 0 0 0 0>
     outlineSettings {
      bUseBrushTransparency=false
      roundingType=HalfHeightRadius
      width=0
      color {colorUseRule=UseColor_Specified, specifiedColor=<rgba 0 0 0 0>}
      cornerRadii=<0 0 0 0>
     }
     tintColor {colorUseRule=UseColor_Specified, specifiedColor=<rgba 1 1 1 1>}
     uVRegion {bIsValid=false, max=<0 0>, min=<0 0>}
    }
    colorAndOpacity=<rgba 0.0395 1 0.2307 1>
   }
   slot {class=CommonVisibilitySwitcherSlot}
  }
  Bad {
   widgetClassPath=&Image
   bIsVariable=true
   parent=InputPin_StateSwitcher
   widget=Bad
   properties {
    visibility=SelfHitTestInvisible
    brush {
     bIsDynamicallyLoaded=false
     drawAs=Image
     imageType=FullColor
     mirroring=NoMirror
     resourceName=None
     resourceObject=/FortUILibrary/Textures/MaterialInstances/MI_UIKit_Icon_ExclamationMark
     tiling=NoTile
     imageSize=<56 56>
     margin=<ltrb 0 0 0 0>
     outlineSettings {
      bUseBrushTransparency=false
      roundingType=HalfHeightRadius
      width=0
      color {colorUseRule=UseColor_Specified, specifiedColor=<rgba 0 0 0 0>}
      cornerRadii=<0 0 0 0>
     }
     tintColor {colorUseRule=UseColor_Specified, specifiedColor=<rgba 1 1 1 1>}
     uVRegion {bIsValid=false, max=<0 0>, min=<0 0>}
    }
    colorAndOpacity=<rgba 1 0.147 0.1329 1>
   }
   slot {class=CommonVisibilitySwitcherSlot}
  }
  Text_Results {
   widgetClassPath=&CommonTextBlock
   bInherited=true
   bIsVariable=true
   parent=LeftSide_Panel
   widget=Text_Results
   properties {
    autoWrapText=true
    text="Incorrect PIN, please try again"
    colorAndOpacity {colorUseRule=UseColor_Specified, specifiedColor=<rgba 1 0.147 0.1329 1>}
    font {
     bForceMonospaced=false
     bMaterialIsStencil=false
     fontMaterial=null
     fontObject=&BurbankSmall
     letterSpacing=0
     monospacedWidth=1
     size=16
     skewAmount=0
     typefaceFontName=Bold
     outlineSettings {
      bApplyOutlineToDropShadows=false
      bMiteredCorners=false
      bSeparateFillAlpha=false
      outlineMaterial=null
      outlineSize=0
      outlineColor=<rgba 0 0 0 1>
     }
    }
    shadowOffset=<0 0>
   }
   slot {class=VerticalBoxSlot, properties {verticalAlignment=VAlign_Top, padding=<ltrb 0 8 0 0>}}
  }
  Button_Accept {
   widgetClassPath=&WBP_UIKit_Button_Regular
   bInherited=true
   bIsVariable=true
   parent=LeftSide_Panel
   widget=Button_Accept
   properties {minWidth=256, text=Accept, triggeringInputAction {dataTable=&/PinModalInputActions, rowName=Accept}}
   slot {class=VerticalBoxSlot, properties {horizontalAlignment=HAlign_Left, verticalAlignment=VAlign_Top, padding=<ltrb 0 24 0 0>}}
  }
  InfoSwitcher {
   widgetClassPath=&CommonVisibilitySwitcher
   bIsVariable=true
   parent=LeftSide_Panel
   widget=InfoSwitcher
   properties {visibility=Collapsed}
   slot {class=VerticalBoxSlot, properties {verticalAlignment=VAlign_Bottom, padding=<ltrb 0 48 0 0>, size {sizeRule=Fill, value=1}}}
  }
  SqueegeeSlot_UnlockInfo {
   widgetClassPath=/Game/UI/Plugin/Squeegee/SqueegeeInjectionSlot
   bIsVariable=true
   parent=InfoSwitcher
   widget=SqueegeeSlot_UnlockInfo
   properties {pluginId=Squeegee, pluginWidgetId=Pin Modal Info}
   slot {class=CommonVisibilitySwitcherSlot}
  }
  PurchaseInfo {
   widgetClassPath=/Game/Athena/UI/Frontend/CosmeticGating/WBP_CosmeticGate_PurchaseInfo
   bIsVariable=true
   parent=InfoSwitcher
   widget=PurchaseInfo
   slot {class=CommonVisibilitySwitcherSlot}
  }
  InfoText {
   widgetClassPath=&/WBP_PinModal_InfoText
   bIsVariable=true
   parent=InfoSwitcher
   widget=InfoText
   slot {class=CommonVisibilitySwitcherSlot}
  }
  RightSide_Panel {widgetClassPath=&Overlay, parent=DefaultLayout, widget=RightSide_Panel, slot {class=HorizontalBoxSlot}}
  Hidden_PasscodeButton_Grid {
   widgetClassPath=/Script/UMG.UniformGridPanel
   parent=RightSide_Panel
   widget=Hidden_PasscodeButton_Grid
   slot {class=OverlaySlot, properties {horizontalAlignment=HAlign_Center, verticalAlignment=VAlign_Center}}
  }
  Hidden_PasscodeButton_1 {
   widgetClassPath=&/PinCodeButton
   bIsVariable=true
   parent=Hidden_PasscodeButton_Grid
   widget=Hidden_PasscodeButton_1
   properties {
    bApplyAlphaOnDisable=true
    onButtonBaseClicked=(&self:WidgetTree.Hidden_PasscodeButton_1.HandleOnButtonClicked)
    onButtonBaseDoubleClicked=(&self:WidgetTree.Hidden_PasscodeButton_1.HandleOnButtonDoubleClicked)
    onButtonBaseHovered=(&self:WidgetTree.Hidden_PasscodeButton_1.HandleOnButtonHovered)
    onButtonBaseUnhovered=(&self:WidgetTree.Hidden_PasscodeButton_1.HandleOnButtonUnhovered)
    onSelectedChangedBase=(&self:WidgetTree.Hidden_PasscodeButton_1.HandleOnSelectedChanged)
    pinValue=1
    triggeringInputAction {dataTable=&PinModalInputActions, rowName=One_Num}
   }
   slot {class=UniformGridSlot, properties {horizontalAlignment=HAlign_Center, verticalAlignment=VAlign_Center}}
  }
  Hidden_PasscodeButton_2 {
   widgetClassPath=&PinCodeButton
   bIsVariable=true
   parent=Hidden_PasscodeButton_Grid
   widget=Hidden_PasscodeButton_2
   properties {
    bApplyAlphaOnDisable=true
    onButtonBaseClicked=(&self:WidgetTree.Hidden_PasscodeButton_2.HandleOnButtonClicked)
    onButtonBaseDoubleClicked=(&self:WidgetTree.Hidden_PasscodeButton_2.HandleOnButtonDoubleClicked)
    onButtonBaseHovered=(&self:WidgetTree.Hidden_PasscodeButton_2.HandleOnButtonHovered)
    onButtonBaseUnhovered=(&self:WidgetTree.Hidden_PasscodeButton_2.HandleOnButtonUnhovered)
    onSelectedChangedBase=(&self:WidgetTree.Hidden_PasscodeButton_2.HandleOnSelectedChanged)
    pinValue=2
    triggeringInputAction {dataTable=&PinModalInputActions, rowName=Two_Num}
   }
   slot {class=UniformGridSlot, properties {column=1, horizontalAlignment=HAlign_Center, verticalAlignment=VAlign_Center}}
  }
  Hidden_PasscodeButton_3 {
   widgetClassPath=&PinCodeButton
   bIsVariable=true
   parent=Hidden_PasscodeButton_Grid
   widget=Hidden_PasscodeButton_3
   properties {
    bApplyAlphaOnDisable=true
    onButtonBaseClicked=(&self:WidgetTree.Hidden_PasscodeButton_3.HandleOnButtonClicked)
    onButtonBaseDoubleClicked=(&self:WidgetTree.Hidden_PasscodeButton_3.HandleOnButtonDoubleClicked)
    onButtonBaseHovered=(&self:WidgetTree.Hidden_PasscodeButton_3.HandleOnButtonHovered)
    onButtonBaseUnhovered=(&self:WidgetTree.Hidden_PasscodeButton_3.HandleOnButtonUnhovered)
    onSelectedChangedBase=(&self:WidgetTree.Hidden_PasscodeButton_3.HandleOnSelectedChanged)
    pinValue=3
    triggeringInputAction {dataTable=&PinModalInputActions, rowName=Three_Num}
   }
   slot {class=UniformGridSlot, properties {column=2, horizontalAlignment=HAlign_Center, verticalAlignment=VAlign_Center}}
  }
  Hidden_PasscodeButton_4 {
   widgetClassPath=&PinCodeButton
   bIsVariable=true
   parent=Hidden_PasscodeButton_Grid
   widget=Hidden_PasscodeButton_4
   properties {
    bApplyAlphaOnDisable=true
    onButtonBaseClicked=(&self:WidgetTree.Hidden_PasscodeButton_4.HandleOnButtonClicked)
    onButtonBaseDoubleClicked=(&self:WidgetTree.Hidden_PasscodeButton_4.HandleOnButtonDoubleClicked)
    onButtonBaseHovered=(&self:WidgetTree.Hidden_PasscodeButton_4.HandleOnButtonHovered)
    onButtonBaseUnhovered=(&self:WidgetTree.Hidden_PasscodeButton_4.HandleOnButtonUnhovered)
    onSelectedChangedBase=(&self:WidgetTree.Hidden_PasscodeButton_4.HandleOnSelectedChanged)
    pinValue=4
    triggeringInputAction {dataTable=&PinModalInputActions, rowName=Four_Num}
   }
   slot {class=UniformGridSlot, properties {horizontalAlignment=HAlign_Center, row=1, verticalAlignment=VAlign_Center}}
  }
  Hidden_PasscodeButton_5 {
   widgetClassPath=&PinCodeButton
   bIsVariable=true
   parent=Hidden_PasscodeButton_Grid
   widget=Hidden_PasscodeButton_5
   properties {
    bApplyAlphaOnDisable=true
    onButtonBaseClicked=(&self:WidgetTree.Hidden_PasscodeButton_5.HandleOnButtonClicked)
    onButtonBaseDoubleClicked=(&self:WidgetTree.Hidden_PasscodeButton_5.HandleOnButtonDoubleClicked)
    onButtonBaseHovered=(&self:WidgetTree.Hidden_PasscodeButton_5.HandleOnButtonHovered)
    onButtonBaseUnhovered=(&self:WidgetTree.Hidden_PasscodeButton_5.HandleOnButtonUnhovered)
    onSelectedChangedBase=(&self:WidgetTree.Hidden_PasscodeButton_5.HandleOnSelectedChanged)
    pinValue=5
    triggeringInputAction {dataTable=&PinModalInputActions, rowName=Five_Num}
   }
   slot {class=UniformGridSlot, properties {column=1, horizontalAlignment=HAlign_Center, row=1, verticalAlignment=VAlign_Center}}
  }
  Hidden_PasscodeButton_6 {
   widgetClassPath=&PinCodeButton
   bIsVariable=true
   parent=Hidden_PasscodeButton_Grid
   widget=Hidden_PasscodeButton_6
   properties {
    bApplyAlphaOnDisable=true
    onButtonBaseClicked=(&self:WidgetTree.Hidden_PasscodeButton_6.HandleOnButtonClicked)
    onButtonBaseDoubleClicked=(&self:WidgetTree.Hidden_PasscodeButton_6.HandleOnButtonDoubleClicked)
    onButtonBaseHovered=(&self:WidgetTree.Hidden_PasscodeButton_6.HandleOnButtonHovered)
    onButtonBaseUnhovered=(&self:WidgetTree.Hidden_PasscodeButton_6.HandleOnButtonUnhovered)
    onSelectedChangedBase=(&self:WidgetTree.Hidden_PasscodeButton_6.HandleOnSelectedChanged)
    pinValue=6
    triggeringInputAction {dataTable=&PinModalInputActions, rowName=Six_Num}
   }
   slot {class=UniformGridSlot, properties {column=2, horizontalAlignment=HAlign_Center, row=1, verticalAlignment=VAlign_Center}}
  }
  Hidden_PasscodeButton_7 {
   widgetClassPath=&PinCodeButton
   bIsVariable=true
   parent=Hidden_PasscodeButton_Grid
   widget=Hidden_PasscodeButton_7
   properties {
    bApplyAlphaOnDisable=true
    onButtonBaseClicked=(&self:WidgetTree.Hidden_PasscodeButton_7.HandleOnButtonClicked)
    onButtonBaseDoubleClicked=(&self:WidgetTree.Hidden_PasscodeButton_7.HandleOnButtonDoubleClicked)
    onButtonBaseHovered=(&self:WidgetTree.Hidden_PasscodeButton_7.HandleOnButtonHovered)
    onButtonBaseUnhovered=(&self:WidgetTree.Hidden_PasscodeButton_7.HandleOnButtonUnhovered)
    onSelectedChangedBase=(&self:WidgetTree.Hidden_PasscodeButton_7.HandleOnSelectedChanged)
    pinValue=7
    triggeringInputAction {dataTable=&PinModalInputActions, rowName=Seven_Num}
   }
   slot {class=UniformGridSlot, properties {horizontalAlignment=HAlign_Center, row=2, verticalAlignment=VAlign_Center}}
  }
  Hidden_PasscodeButton_8 {
   widgetClassPath=&PinCodeButton
   bIsVariable=true
   parent=Hidden_PasscodeButton_Grid
   widget=Hidden_PasscodeButton_8
   properties {
    bApplyAlphaOnDisable=true
    onButtonBaseClicked=(&self:WidgetTree.Hidden_PasscodeButton_8.HandleOnButtonClicked)
    onButtonBaseDoubleClicked=(&self:WidgetTree.Hidden_PasscodeButton_8.HandleOnButtonDoubleClicked)
    onButtonBaseHovered=(&self:WidgetTree.Hidden_PasscodeButton_8.HandleOnButtonHovered)
    onButtonBaseUnhovered=(&self:WidgetTree.Hidden_PasscodeButton_8.HandleOnButtonUnhovered)
    onSelectedChangedBase=(&self:WidgetTree.Hidden_PasscodeButton_8.HandleOnSelectedChanged)
    pinValue=8
    triggeringInputAction {dataTable=&PinModalInputActions, rowName=Eight_Num}
   }
   slot {class=UniformGridSlot, properties {column=1, horizontalAlignment=HAlign_Center, row=2, verticalAlignment=VAlign_Center}}
  }
  Hidden_PasscodeButton_9 {
   widgetClassPath=&PinCodeButton
   bIsVariable=true
   parent=Hidden_PasscodeButton_Grid
   widget=Hidden_PasscodeButton_9
   properties {
    bApplyAlphaOnDisable=true
    onButtonBaseClicked=(&self:WidgetTree.Hidden_PasscodeButton_9.HandleOnButtonClicked)
    onButtonBaseDoubleClicked=(&self:WidgetTree.Hidden_PasscodeButton_9.HandleOnButtonDoubleClicked)
    onButtonBaseHovered=(&self:WidgetTree.Hidden_PasscodeButton_9.HandleOnButtonHovered)
    onButtonBaseUnhovered=(&self:WidgetTree.Hidden_PasscodeButton_9.HandleOnButtonUnhovered)
    onSelectedChangedBase=(&self:WidgetTree.Hidden_PasscodeButton_9.HandleOnSelectedChanged)
    pinValue=9
    triggeringInputAction {dataTable=&PinModalInputActions, rowName=Nine_Num}
   }
   slot {class=UniformGridSlot, properties {column=2, horizontalAlignment=HAlign_Center, row=2, verticalAlignment=VAlign_Center}}
  }
  Hidden_PasscodeButton_0 {
   widgetClassPath=&PinCodeButton
   bIsVariable=true
   parent=Hidden_PasscodeButton_Grid
   widget=Hidden_PasscodeButton_0
   properties {
    bApplyAlphaOnDisable=true
    onButtonBaseClicked=(&self:WidgetTree.Hidden_PasscodeButton_0.HandleOnButtonClicked)
    onButtonBaseDoubleClicked=(&self:WidgetTree.Hidden_PasscodeButton_0.HandleOnButtonDoubleClicked)
    onButtonBaseHovered=(&self:WidgetTree.Hidden_PasscodeButton_0.HandleOnButtonHovered)
    onButtonBaseUnhovered=(&self:WidgetTree.Hidden_PasscodeButton_0.HandleOnButtonUnhovered)
    onSelectedChangedBase=(&self:WidgetTree.Hidden_PasscodeButton_0.HandleOnSelectedChanged)
    triggeringInputAction {dataTable=&PinModalInputActions, rowName=Zero_Num}
   }
   slot {class=UniformGridSlot, properties {column=1, horizontalAlignment=HAlign_Center, row=3, verticalAlignment=VAlign_Center}}
  }
  RightSide_Bg {
   widgetClassPath=&Image
   parent=RightSide_Panel
   widget=RightSide_Bg
   properties {
    visibility=SelfHitTestInvisible
    brush {
     bIsDynamicallyLoaded=false
     drawAs=Image
     imageType=FullColor
     mirroring=NoMirror
     resourceName=None
     resourceObject=/Game/UI/Foundation/Modals/PasscodeModal/Materials/MI_UI_PinModal_Background_Side
     tiling=NoTile
     imageSize=<32 32>
     margin=<ltrb 0 0 0 0>
     outlineSettings {
      bUseBrushTransparency=false
      roundingType=HalfHeightRadius
      width=0
      color {colorUseRule=UseColor_Specified, specifiedColor=<rgba 0 0 0 0>}
      cornerRadii=<0 0 0 0>
     }
     tintColor {colorUseRule=UseColor_Specified, specifiedColor=<rgba 1 1 1 1>}
     uVRegion {bIsValid=false, max=<0 0>, min=<0 0>}
    }
   }
   slot {class=OverlaySlot, properties {horizontalAlignment=HAlign_Fill, verticalAlignment=VAlign_Fill}}
  }
  VerticalBox_0 {
   widgetClassPath=&VerticalBox
   parent=RightSide_Panel
   widget=VerticalBox_0
   slot {class=OverlaySlot, properties {horizontalAlignment=HAlign_Center, verticalAlignment=VAlign_Fill, padding=<ltrb 0 32 0 32>}}
  }
  PasscodeButton_Grid {
   widgetClassPath=&UniformGridPanel
   parent=VerticalBox_0
   widget=PasscodeButton_Grid
   slot {class=VerticalBoxSlot, properties {verticalAlignment=VAlign_Center, padding=<ltrb 48 40 48 40>, size {sizeRule=Fill, value=1}}}
  }
  PasscodeButton_1 {
   widgetClassPath=&PinCodeButton
   bIsVariable=true
   parent=PasscodeButton_Grid
   widget=PasscodeButton_1
   properties {
    bApplyAlphaOnDisable=true
    onButtonBaseClicked=(&self:WidgetTree.PasscodeButton_1.HandleOnButtonClicked)
    onButtonBaseDoubleClicked=(&self:WidgetTree.PasscodeButton_1.HandleOnButtonDoubleClicked)
    onButtonBaseHovered=(&self:WidgetTree.PasscodeButton_1.HandleOnButtonHovered)
    onButtonBaseUnhovered=(&self:WidgetTree.PasscodeButton_1.HandleOnButtonUnhovered)
    onSelectedChangedBase=(&self:WidgetTree.PasscodeButton_1.HandleOnSelectedChanged)
    pinValue=1
    triggeringInputAction {dataTable=&PinModalInputActions, rowName=One}
   }
   slot {class=UniformGridSlot, properties {horizontalAlignment=HAlign_Center, verticalAlignment=VAlign_Center}}
  }
  PasscodeButton_2 {
   widgetClassPath=&PinCodeButton
   bIsVariable=true
   parent=PasscodeButton_Grid
   widget=PasscodeButton_2
   properties {
    bApplyAlphaOnDisable=true
    onButtonBaseClicked=(&self:WidgetTree.PasscodeButton_2.HandleOnButtonClicked)
    onButtonBaseDoubleClicked=(&self:WidgetTree.PasscodeButton_2.HandleOnButtonDoubleClicked)
    onButtonBaseHovered=(&self:WidgetTree.PasscodeButton_2.HandleOnButtonHovered)
    onButtonBaseUnhovered=(&self:WidgetTree.PasscodeButton_2.HandleOnButtonUnhovered)
    onSelectedChangedBase=(&self:WidgetTree.PasscodeButton_2.HandleOnSelectedChanged)
    pinValue=2
    triggeringInputAction {dataTable=&PinModalInputActions, rowName=Two}
   }
   slot {class=UniformGridSlot, properties {column=1, horizontalAlignment=HAlign_Center, verticalAlignment=VAlign_Center}}
  }
  PasscodeButton_3 {
   widgetClassPath=&PinCodeButton
   bIsVariable=true
   parent=PasscodeButton_Grid
   widget=PasscodeButton_3
   properties {
    bApplyAlphaOnDisable=true
    onButtonBaseClicked=(&self:WidgetTree.PasscodeButton_3.HandleOnButtonClicked)
    onButtonBaseDoubleClicked=(&self:WidgetTree.PasscodeButton_3.HandleOnButtonDoubleClicked)
    onButtonBaseHovered=(&self:WidgetTree.PasscodeButton_3.HandleOnButtonHovered)
    onButtonBaseUnhovered=(&self:WidgetTree.PasscodeButton_3.HandleOnButtonUnhovered)
    onSelectedChangedBase=(&self:WidgetTree.PasscodeButton_3.HandleOnSelectedChanged)
    pinValue=3
    triggeringInputAction {dataTable=&PinModalInputActions, rowName=Three}
   }
   slot {class=UniformGridSlot, properties {column=2, horizontalAlignment=HAlign_Center, verticalAlignment=VAlign_Center}}
  }
  PasscodeButton_4 {
   widgetClassPath=&PinCodeButton
   bIsVariable=true
   parent=PasscodeButton_Grid
   widget=PasscodeButton_4
   properties {
    bApplyAlphaOnDisable=true
    onButtonBaseClicked=(&self:WidgetTree.PasscodeButton_4.HandleOnButtonClicked)
    onButtonBaseDoubleClicked=(&self:WidgetTree.PasscodeButton_4.HandleOnButtonDoubleClicked)
    onButtonBaseHovered=(&self:WidgetTree.PasscodeButton_4.HandleOnButtonHovered)
    onButtonBaseUnhovered=(&self:WidgetTree.PasscodeButton_4.HandleOnButtonUnhovered)
    onSelectedChangedBase=(&self:WidgetTree.PasscodeButton_4.HandleOnSelectedChanged)
    pinValue=4
    triggeringInputAction {dataTable=&PinModalInputActions, rowName=Four}
   }
   slot {class=UniformGridSlot, properties {horizontalAlignment=HAlign_Center, row=1, verticalAlignment=VAlign_Center}}
  }
  PasscodeButton_5 {
   widgetClassPath=&PinCodeButton
   bIsVariable=true
   parent=PasscodeButton_Grid
   widget=PasscodeButton_5
   properties {
    bApplyAlphaOnDisable=true
    onButtonBaseClicked=(&self:WidgetTree.PasscodeButton_5.HandleOnButtonClicked)
    onButtonBaseDoubleClicked=(&self:WidgetTree.PasscodeButton_5.HandleOnButtonDoubleClicked)
    onButtonBaseHovered=(&self:WidgetTree.PasscodeButton_5.HandleOnButtonHovered)
    onButtonBaseUnhovered=(&self:WidgetTree.PasscodeButton_5.HandleOnButtonUnhovered)
    onSelectedChangedBase=(&self:WidgetTree.PasscodeButton_5.HandleOnSelectedChanged)
    pinValue=5
    triggeringInputAction {dataTable=&PinModalInputActions, rowName=Five}
   }
   slot {class=UniformGridSlot, properties {column=1, horizontalAlignment=HAlign_Center, row=1, verticalAlignment=VAlign_Center}}
  }
  PasscodeButton_6 {
   widgetClassPath=&PinCodeButton
   bIsVariable=true
   parent=PasscodeButton_Grid
   widget=PasscodeButton_6
   properties {
    bApplyAlphaOnDisable=true
    onButtonBaseClicked=(&self:WidgetTree.PasscodeButton_6.HandleOnButtonClicked)
    onButtonBaseDoubleClicked=(&self:WidgetTree.PasscodeButton_6.HandleOnButtonDoubleClicked)
    onButtonBaseHovered=(&self:WidgetTree.PasscodeButton_6.HandleOnButtonHovered)
    onButtonBaseUnhovered=(&self:WidgetTree.PasscodeButton_6.HandleOnButtonUnhovered)
    onSelectedChangedBase=(&self:WidgetTree.PasscodeButton_6.HandleOnSelectedChanged)
    pinValue=6
    triggeringInputAction {dataTable=&PinModalInputActions, rowName=Six}
   }
   slot {class=UniformGridSlot, properties {column=2, horizontalAlignment=HAlign_Center, row=1, verticalAlignment=VAlign_Center}}
  }
  PasscodeButton_7 {
   widgetClassPath=&PinCodeButton
   bIsVariable=true
   parent=PasscodeButton_Grid
   widget=PasscodeButton_7
   properties {
    bApplyAlphaOnDisable=true
    onButtonBaseClicked=(&self:WidgetTree.PasscodeButton_7.HandleOnButtonClicked)
    onButtonBaseDoubleClicked=(&self:WidgetTree.PasscodeButton_7.HandleOnButtonDoubleClicked)
    onButtonBaseHovered=(&self:WidgetTree.PasscodeButton_7.HandleOnButtonHovered)
    onButtonBaseUnhovered=(&self:WidgetTree.PasscodeButton_7.HandleOnButtonUnhovered)
    onSelectedChangedBase=(&self:WidgetTree.PasscodeButton_7.HandleOnSelectedChanged)
    pinValue=7
    triggeringInputAction {dataTable=&PinModalInputActions, rowName=Seven}
   }
   slot {class=UniformGridSlot, properties {horizontalAlignment=HAlign_Center, row=2, verticalAlignment=VAlign_Center}}
  }
  PasscodeButton_8 {
   widgetClassPath=&PinCodeButton
   bIsVariable=true
   parent=PasscodeButton_Grid
   widget=PasscodeButton_8
   properties {
    bApplyAlphaOnDisable=true
    onButtonBaseClicked=(&self:WidgetTree.PasscodeButton_8.HandleOnButtonClicked)
    onButtonBaseDoubleClicked=(&self:WidgetTree.PasscodeButton_8.HandleOnButtonDoubleClicked)
    onButtonBaseHovered=(&self:WidgetTree.PasscodeButton_8.HandleOnButtonHovered)
    onButtonBaseUnhovered=(&self:WidgetTree.PasscodeButton_8.HandleOnButtonUnhovered)
    onSelectedChangedBase=(&self:WidgetTree.PasscodeButton_8.HandleOnSelectedChanged)
    pinValue=8
    triggeringInputAction {dataTable=&PinModalInputActions, rowName=Eight}
   }
   slot {class=UniformGridSlot, properties {column=1, horizontalAlignment=HAlign_Center, row=2, verticalAlignment=VAlign_Center}}
  }
  PasscodeButton_9 {
   widgetClassPath=&PinCodeButton
   bIsVariable=true
   parent=PasscodeButton_Grid
   widget=PasscodeButton_9
   properties {
    bApplyAlphaOnDisable=true
    onButtonBaseClicked=(&self:WidgetTree.PasscodeButton_9.HandleOnButtonClicked)
    onButtonBaseDoubleClicked=(&self:WidgetTree.PasscodeButton_9.HandleOnButtonDoubleClicked)
    onButtonBaseHovered=(&self:WidgetTree.PasscodeButton_9.HandleOnButtonHovered)
    onButtonBaseUnhovered=(&self:WidgetTree.PasscodeButton_9.HandleOnButtonUnhovered)
    onSelectedChangedBase=(&self:WidgetTree.PasscodeButton_9.HandleOnSelectedChanged)
    pinValue=9
    triggeringInputAction {dataTable=&PinModalInputActions, rowName=Nine}
   }
   slot {class=UniformGridSlot, properties {column=2, horizontalAlignment=HAlign_Center, row=2, verticalAlignment=VAlign_Center}}
  }
  PasscodeButton_0 {
   widgetClassPath=&PinCodeButton
   bIsVariable=true
   parent=PasscodeButton_Grid
   widget=PasscodeButton_0
   properties {
    bApplyAlphaOnDisable=true
    onButtonBaseClicked=(&self:WidgetTree.PasscodeButton_0.HandleOnButtonClicked)
    onButtonBaseDoubleClicked=(&self:WidgetTree.PasscodeButton_0.HandleOnButtonDoubleClicked)
    onButtonBaseHovered=(&self:WidgetTree.PasscodeButton_0.HandleOnButtonHovered)
    onButtonBaseUnhovered=(&self:WidgetTree.PasscodeButton_0.HandleOnButtonUnhovered)
    onSelectedChangedBase=(&self:WidgetTree.PasscodeButton_0.HandleOnSelectedChanged)
    triggeringInputAction {dataTable=&PinModalInputActions, rowName=Zero}
   }
   slot {class=UniformGridSlot, properties {column=1, horizontalAlignment=HAlign_Center, row=3, verticalAlignment=VAlign_Center}}
  }
  VisibilityBorder_Backspace_Touch {
   widgetClassPath=&CommonHardwareVisibilityBorder
   parent=PasscodeButton_Grid
   widget=VisibilityBorder_Backspace_Touch
   properties {
    horizontalAlignment=HAlign_Center
    verticalAlignment=VAlign_Center
    visibilityQuery {
     autoDescription=" ANY( Input.Touch )"
     tokenStreamVersion=0
     userDescription=while using touch
     queryTokenStream [0,1,1,1,0]
     tagDictionary [{tagName=Input.Touch}]
    }
   }
   slot {class=UniformGridSlot, properties {column=2, horizontalAlignment=HAlign_Center, row=3, verticalAlignment=VAlign_Center}}
  }
  Button_Backspace {
   widgetClassPath=/FortUILibrary/UIKit/Buttons/WBP_UIKit_Button_Tab_Enhanced
   bIsVariable=true
   parent=VisibilityBorder_Backspace_Touch
   widget=Button_Backspace
   properties {
    bSelectable=false
    icon=/FortUILibrary/Textures/Icons/T_UI_Icon_UIKit_Backspace
    useBackgroundBlock=false
    useTextBlock=false
    iconSize=<64 64>
   }
   slot {class=BorderSlot, properties {horizontalAlignment=HAlign_Center, verticalAlignment=VAlign_Center, padding=<ltrb 0 0 0 0>}}
  }
  PinButtons {
   widgetClassPath=&HorizontalBox
   parent=VerticalBox_0
   widget=PinButtons
   slot {class=VerticalBoxSlot, properties {horizontalAlignment=HAlign_Center, verticalAlignment=VAlign_Bottom, padding=<ltrb 40 0 40 0>}}
  }
  Button_ForgotPin {
   widgetClassPath=/FortUILibrary/UIKit/Buttons/WBP_UIKit_Button_Quiet
   bIsVariable=true
   parent=PinButtons
   widget=Button_ForgotPin
   properties {
    bIsFocusable=false
    text=Forgot Pin
    triggeringInputAction {dataTable=&PinModalInputActions, rowName=ForgotPin}
   }
   slot {class=HorizontalBoxSlot, properties {padding=<ltrb 0 0 24 0>}}
  }
  Button_Close {
   widgetClassPath=&WBP_UIKit_Button_Quiet
   bIsVariable=true
   parent=PinButtons
   widget=Button_Close
   properties {
    bIsFocusable=false
    text=Close
    triggeringInputAction {dataTable=&PinModalInputActions, rowName=Close}
   }
   slot {class=HorizontalBoxSlot}
  }
  VisibilityBorder_GamepadOnly {
   widgetClassPath=&CommonHardwareVisibilityBorder
   parent=VerticalBox_0
   widget=VisibilityBorder_GamepadOnly
   properties {
    padding=<ltrb 0 24 0 0>
    visibilityQuery {
     autoDescription=" ANY( Input.Gamepad )"
     tokenStreamVersion=0
     userDescription=OnlyGamepad
     queryTokenStream [0,1,1,1,0]
     tagDictionary [{tagName=Input.Gamepad}]
    }
   }
   slot {class=VerticalBoxSlot, properties {horizontalAlignment=HAlign_Center, verticalAlignment=VAlign_Bottom}}
  }
  Button_AboutRatings {
   widgetClassPath=&WBP_UIKit_Button_Quiet
   bIsVariable=true
   parent=VisibilityBorder_GamepadOnly
   widget=Button_AboutRatings
   properties {
    bIsFocusable=false
    text=About Ratings
    visibility=Collapsed
    triggeringInputAction {dataTable=&InputActionDataTable, rowName=GenericLeftStickClick}
   }
   slot {class=BorderSlot, properties {padding=<ltrb 0 24 0 0>}}
  }
  CabinLayout {
   widgetClassPath=&VerticalBox
   bIsVariable=true
   parent=CabinSwitcher
   widget=CabinLayout
   slot {class=WidgetSwitcherSlot, properties {padding=<ltrb 50 50 50 50>}}
  }
  HeaderText_1 {
   widgetClassPath=&CommonTextBlock
   parent=CabinLayout
   widget=HeaderText_1
   properties {
    text=Your Account Needs to be Set up
    textTransformPolicy=ToUpper
    font {
     bForceMonospaced=false
     bMaterialIsStencil=false
     fontMaterial=null
     fontObject=&Burbank
     letterSpacing=0
     monospacedWidth=1
     size=54
     skewAmount=0
     typefaceFontName=Burbank-Black
     outlineSettings {
      bApplyOutlineToDropShadows=false
      bMiteredCorners=false
      bSeparateFillAlpha=false
      outlineMaterial=null
      outlineSize=0
      outlineColor=<rgba 0 0 0 1>
     }
    }
    shadowOffset=<0 0>
   }
   slot {class=VerticalBoxSlot, properties {horizontalAlignment=HAlign_Center, verticalAlignment=VAlign_Top, padding=<ltrb 0 0 0 64>}}
  }
  EmailManagementWidget {
   widgetClassPath=/Game/UI/Foundation/CabinMode/ParentalControlsCabinEmailManagementWidget
   bInherited=true
   bIsVariable=true
   parent=CabinLayout
   widget=EmailManagementWidget
   slot {class=VerticalBoxSlot, properties {verticalAlignment=VAlign_Top, size {sizeRule=Fill, value=1}}}
  }
  Button_Close_1 {
   widgetClassPath=&IconTextButton
   bIsVariable=true
   parent=CabinLayout
   widget=Button_Close_1
   properties {
    buttonText=Close
    contentAlignment=HAlign_Center
    controllerInputStyle=/FortUILibrary/Foundation/Buttons/ButtonStyles/ButtonStyle-BottomBar-Console.ButtonStyle-BottomBar-Console_C
    onButtonBaseClicked=(&self:WidgetTree.Button_Close_1.HandleOnButtonClicked)
    onButtonBaseDoubleClicked=(&self:WidgetTree.Button_Close_1.HandleOnButtonDoubleClicked)
    onButtonBaseHovered=(&self:WidgetTree.Button_Close_1.HandleOnButtonHovered)
    onButtonBaseUnhovered=(&self:WidgetTree.Button_Close_1.HandleOnButtonUnhovered)
    onSelectedChangedBase=(&self:WidgetTree.Button_Close_1.HandleOnSelectedChanged)
    style=/FortUILibrary/Foundation/Buttons/ButtonStyles/Tab/ButtonStyle-SoloButton.ButtonStyle-SoloButton_C
    triggeringInputAction {dataTable=&PinModalInputActions, rowName=Close}
   }
   slot {class=VerticalBoxSlot, properties {horizontalAlignment=HAlign_Right, verticalAlignment=VAlign_Bottom}}
  }
  SqueegeeInjectionSlot {widgetClassPath=/Script/FortniteUI.FortSqueegeeWidgetInjectionSlot, bInherited=true, widget=null}
 ]
}
