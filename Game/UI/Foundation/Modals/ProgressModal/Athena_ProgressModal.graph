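# Widget Blueprint for a progress modal that shows a throbber beside title and description text, based on the native AthenaProgressModal class.
# Its root is the shared UIKit dialog, whose named slot hosts a horizontal box with the throbber and text column.
# Title, description and a Cancel button, collapsed by default, are inherited widgets from the parent class.
# On Construct, the graph only collapses the dialog's Content_FirstColumn grid panel.
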WidgetBlueprint /Game/UI/Foundation/Modals/ProgressModal/Athena_ProgressModal
graphs [EventGraph {nodes [
   Event Construct()
   > Set Visibility(
    Visibility=Collapsed
    self=Get Content_FirstColumn!(self:WBP UIKit Dialog Internal Object Reference=Get Dialog_Internal() -> (Dialog_Internal:WBP UIKit Dialog Internal Object Reference)) -> (Content_FirstColumn:Grid Panel Object Reference)
   )
  ]}]
variables []
widgets {
 info {
  inheritedWidgetCount=3
  namedSlotCount=1
  parentClass=/Script/FortniteUI.AthenaProgressModal
  rootWidgetClass=/FortUILibrary/Systems/Dialog/WBP_UIKit_Dialog_Internal.WBP_UIKit_Dialog_Internal_C
  widgetCount=7
 }
 widgets [
  Dialog_Internal {
   widgetClassPath=/FortUILibrary/Systems/Dialog/WBP_UIKit_Dialog_Internal
   bIsVariable=true
   widget=Dialog_Internal
   properties {dialogSIze {
     height Override {enable=false, value=800}
     max Aspect Ratio {enable=false, value=0}
     max Desired Height {enable=true, value=800}
     max Desired Width {enable=false, value=0}
     min Aspect Ratio {enable=false, value=0}
     min Desired Height {enable=true, value=200}
     min Desired Width {enable=true, value=864}
     width Override {enable=false, value=1200}
    }}
  }
  MainBox {widgetClassPath=&HorizontalBox, namedSlotHost=Dialog_Internal, widget=MainBox}
  WBP_UIKit_Throbber {
   widgetClassPath=/FortUILibrary/UIKit/Throbber/WBP_UIKit_Throbber
   bIsVariable=true
   parent=MainBox
   widget=WBP_UIKit_Throbber
   properties {throbberSize=Medium}
   slot {class=HorizontalBoxSlot, properties {horizontalAlignment=HAlign_Left, verticalAlignment=VAlign_Center, padding=<ltrb 80 0 0 0>}}
  }
  VerticalBox_0 {
   widgetClassPath=&VerticalBox
   parent=MainBox
   widget=VerticalBox_0
   slot {class=HorizontalBoxSlot, properties {padding=<ltrb 60 42 80 42>}}
  }
  Text_Title {
   widgetClassPath=&CommonTextBlock
   bInherited=true
   parent=VerticalBox_0
   widget=Text_Title
   properties {
    autoWrapText=true
    clipping=ClipToBounds
    justification=Center
    lineHeightPercentage=0.9
    scrollStyle=&TextScrollStyle-Base_C
    text="{TITLE} really long text thing"
    textTransformPolicy=ToUpper
    font {
     bForceMonospaced=false
     bMaterialIsStencil=false
     fontMaterial=null
     fontObject=&HeadingNow
     letterSpacing=0
     monospacedWidth=1
     size=41.25
     skewAmount=0
     typefaceFontName=86 Bold Italic
     outlineSettings {
      bApplyOutlineToDropShadows=false
      bMiteredCorners=false
      bSeparateFillAlpha=false
      outlineMaterial=null
      outlineSize=0
      outlineColor=<rgba 0 0 0 1>
     }
    }
    margin=<ltrb 0 0 0 10>
    shadowOffset=<0 0>
   }
   slot {class=VerticalBoxSlot, properties {padding=<ltrb 0 0 0 16>}}
  }
  Text_Description {
   widgetClassPath=&CommonTextBlock
   bInherited=true
   parent=VerticalBox_0
   widget=Text_Description
   properties {
    autoWrapText=true
    clipping=ClipToBounds
    justification=Center
    lineHeightPercentage=0.9
    text="{DESCRIPTION}"
    textTransformPolicy=ToUpper
    colorAndOpacity {colorUseRule=UseColor_Specified, specifiedColor=<rgba 1 1 1 0.7>}
    font {
     bForceMonospaced=false
     bMaterialIsStencil=false
     fontMaterial=null
     fontObject=&HeadingNow
     letterSpacing=0
     monospacedWidth=1
     size=22.5
     skewAmount=0
     typefaceFontName=74 Regular
     outlineSettings {
      bApplyOutlineToDropShadows=false
      bMiteredCorners=false
      bSeparateFillAlpha=false
      outlineMaterial=null
      outlineSize=0
      outlineColor=<rgba 0 0 0 1>
     }
    }
    margin=<ltrb 0 0 0 10>
    shadowOffset=<0 0>
   }
   slot {class=VerticalBoxSlot}
  }
  ActionButton_Cancel {
   widgetClassPath=/Game/UI/Frontend/BottomBar/AthenaHUDBoundActionButton
   bInherited=true
   bIsVariable=true
   parent=VerticalBox_0
   widget=ActionButton_Cancel
   properties {visibility=Collapsed}
   slot {class=VerticalBoxSlot, properties {horizontalAlignment=HAlign_Right}}
  }
 ]
}
