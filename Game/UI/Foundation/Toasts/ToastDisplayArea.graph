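# Widget Blueprint for the frontend toast display area, built on FortToastDisplayArea, that hosts Fortnite notification popups.
# A SafeZone without top or bottom padding wraps an Overlay stacking inherited toast widgets for power, minor errors, social, Battle Pass page unlocks, Deadpool challenges and others.
# A bottom-aligned voice channel popup runs in toast mode.
# The only graph logic is a TestFrontendToast event that fires sample toasts across several notification types.
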WidgetBlueprint /Game/UI/Foundation/Toasts/ToastDisplayArea
graphs [EventGraph {nodes [
   TestFrontendToast() {is=title K2Node_CustomEvent}
   > TestBasicToast!(
    execute
    self:Fort Toast Display Area Object Reference
    InNotificationType:EFortNotificationType Enum=Default
    InTitle:Text="INVTEXT(\"Default Title\")"
    InDescription:Text="LOCTEXT(\"Default description\")"
   ) -> (then)
   > TestBasicToast(InNotificationType=Power, InTitle="INVTEXT(\"Power\")", InDescription="LOCTEXT(\"Power toast widget description\")")
   > TestBasicToast(InNotificationType=HealthWarning, InTitle="INVTEXT(\"Health Warning\")", InDescription="LOCTEXT(\"Health Warning description\")")
   > TestBasicToast(InNotificationType=Error, InTitle="INVTEXT(\"Error\")", InDescription="LOCTEXT(\"Error toast description\")")
   > TestBasicToast(InNotificationType=GiftSent, InTitle="INVTEXT(\"Gift Sent\")", InDescription="LOCTEXT(\"Gift Sent description\")")
   > TestBasicToast(InNotificationType=VoiceChannel, InTitle="INVTEXT(\"Voice Channel\")", InDescription="LOCTEXT(\"Voice Channel description\")")
   > TestBasicToast(
    InNotificationType=DonutChallenge
    InTitle="INVTEXT(\"DEADPOOL CHALLENGE COMPLETED\")"
    InDescription="INVTEXT(\"Visit Deadpools room for more!\")"
   )
   > TestBasicToast(InNotificationType=BattlePassPageUnlock, InTitle="INVTEXT(\"Battle Pass\")", InDescription="LOCTEXT(\"Battle Pass Page Unlock\")")
   > TestBasicToast(
    InNotificationType=DownloadOnDemandError
    InTitle="INVTEXT(\"Download On Demand Error\")"
    InDescription="LOCTEXT(\"Download On Demand Error Description\")"
   )
  ]}]
variables []
widgets {
 info {inheritedWidgetCount=9, parentClass=/Script/FortniteUI.FortToastDisplayArea, rootWidgetClass=&SafeZone, widgetCount=11}
 widgets [
  SafeZone_0 {widgetClassPath=&SafeZone, bIsVariable=true, widget=SafeZone_0, properties {padBottom=false, padTop=false}}
  Overlay_0 {widgetClassPath=&Overlay, parent=SafeZone_0, widget=Overlay_0, slot {class=SafeZoneSlot}}
  ToastWidget_New {
   widgetClassPath=&/ToastWidget_New
   bInherited=true
   parent=Overlay_0
   widget=ToastWidget_New
   slot {class=OverlaySlot, properties {padding=<ltrb 0 200 0 0>}}
  }
  Toast_BattlePassPageUnlocked {
   widgetClassPath=&/BattlePassPageUnlockedToast
   bInherited=true
   bIsVariable=true
   parent=Overlay_0
   widget=Toast_BattlePassPageUnlocked
   slot {class=OverlaySlot, properties {padding=<ltrb 0 200 0 0>}}
  }
  PowerToastWidget {
   widgetClassPath=&/PowerToastWidget
   bInherited=true
   parent=Overlay_0
   widget=PowerToastWidget
   slot {class=OverlaySlot, properties {padding=<ltrb 0 200 0 0>}}
  }
  ToastWidget_MinorError {
   widgetClassPath=&/ToastWidget_MinorError
   bInherited=true
   parent=Overlay_0
   widget=ToastWidget_MinorError
   slot {class=OverlaySlot, properties {padding=<ltrb 0 200 0 0>}}
  }
  VoiceChannelChangePopup {
   widgetClassPath=/Game/UI/InGame/HUD/VoiceChannelChangePopup
   bInherited=true
   parent=Overlay_0
   widget=VoiceChannelChangePopup
   properties {bToastMode=true, visiblityInSeconds=4}
   slot {class=OverlaySlot, properties {verticalAlignment=VAlign_Bottom, padding=<ltrb 85 0 0 85>}}
  }
  ToastWidgetDonut {
   widgetClassPath=&/ToastWidgetDonut
   bInherited=true
   bIsVariable=true
   parent=Overlay_0
   widget=ToastWidgetDonut
   slot {class=OverlaySlot, properties {padding=<ltrb 20 180 0 0>}}
  }
  ToastSocialWidget {
   widgetClassPath=&/ToastSocialWidget
   bInherited=true
   bIsVariable=true
   parent=Overlay_0
   widget=ToastSocialWidget
   slot {class=OverlaySlot, properties {padding=<ltrb 0 200 0 0>}}
  }
  ToastSocialWidget_New {
   widgetClassPath=&/ToastSocialWidget_Base
   bInherited=true
   bIsVariable=true
   parent=Overlay_0
   widget=ToastSocialWidget_New
   slot {class=OverlaySlot, properties {padding=<ltrb 0 200 0 0>}}
  }
  TutorialHighlight {widgetClassPath=&TutorialHighlight}
 ]
}
