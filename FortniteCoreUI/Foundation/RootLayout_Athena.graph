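# Root viewport layout widget for Athena, stacking screen content, a modal queue, toasts, a plugin popup slot and system modals, some lazy-loaded, in one overlay.
# On construct it hooks the window close button to a confirmation dialog that quits on confirm, and binds a visibility-change delegate.
# Visibility changes to SelfHitTestInvisible play a take-control sound, and OnModalQueueEmptied fires client events announcing finished modals.
# IsConsole checks for PS4 or XboxOne.
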WidgetBlueprint /FortniteCoreUI/Foundation/RootLayout_Athena
graphs [
 EventGraph {nodes [
   n1 = VisibilityChanged!() -> (OutputDelegate, then, In Visibility:ESlateVisibility Enum=Visible) {is=title K2Node_CustomEvent}
   > n2 = Switch on ESlateVisibility!(execute, Selection:ESlateVisibility Enum=@n1.In Visibility) -> (Visible:Exec, Collapsed:Exec, Hidden:Exec, HitTestInvisible:Exec, SelfHitTestInvisible:Exec) {is=title K2Node_SwitchEnum}
   PlaySound2D(execute=@n2.SelfHitTestInvisible, Sound=Get TakeControlAudio() -> (TakeControlAudio:Sound Base Object Reference))
   n3 = OnGameWindowCloseButtonClicked() {is=title K2Node_CustomEvent}
   > n4 = ShowSimpleConfirmationDialog_CustomInput!(
    execute:Exec
    Icon:Slate Brush Structure
    Title:Text=Get QuitTitle() -> (QuitTitle:Text)
    Message:Text=Get QuitMessage() -> (QuitMessage:Text)
    ConfirmAction:Name=Confirm
    DeclineAction:Name=Cancel
    AdditionalContent:Widget Object Reference
    LeftAdditionalContent:Widget Object Reference
    Cancelable:bool=false
    TargetPlayer:Local Player Object Reference
   ) -> (then:Exec, DialogResult:Exec, Result:EFortDialogResult Enum, ResultName:Name) {is=title K2Node_AsyncAction}
   Switch on EFortDialogResult!(execute=@n4.DialogResult, Selection:EFortDialogResult Enum=@n4) -> (Confirmed:Exec, Declined:Exec, Ignored:Exec, Killed:Exec, TimedOut:Exec) {is=title K2Node_SwitchEnum}
   > QuitGame(self:Fort Global UIContext Object Reference=Get FortGlobalUIContext!() -> (ReturnValue:Fort Global UIContext Object Reference) {is=title K2Node_GetBlueprintContextEx})
   K2Node_Event_0 = OnModalQueueEmptied_1() {is=title K2Node_CustomEvent}
   > Execute Client Event!(
    execute:Exec
    EventType:Fort Client Event Name Structure=(CategoryName="Homebase",EventName="FinishedModalQueue")
    EventSource:Object Reference
    EventFocus:Object Reference
   ) -> (then:Exec) {is=title K2Node_ExecuteClientEvent}
   > Execute Client Event(EventType=(CategoryName="UserInterface",EventName="FinishedModalWindows"))
   Event Construct()
   > SetWindowTitleBarOnCloseClickedDelegate!(execute, Delegate:Delegate=@n3) -> (then)
   > Bind Event to On Visibility Changed(self:User Widget Object Reference, Delegate=Create Event()) {is=title K2Node_AddDelegate}
  ]}
 IsConsole {nodes [
   IsConsole() {is=title K2Node_FunctionEntry}
   n1 = GetPlatformName!() -> (ReturnValue:String)
   > Return Node!(execute, ReturnValue:bool=OR Boolean(A="Equal Exactly (String)"(A=@n1, B=PS4), B="Equal Exactly (String)"(A=@n1, B=XboxOne)))
  ]}
]
variables [
 bClosingErrorDialog {type=boolean}
 ConfirmationWindow {type=ConfirmationWindow_C, default=null}
 ControllerDisconnectedModal {type=ProgressModalWidget_C, default=null}
 QuitMessage {type=string, default=Are you sure you want to exit Fortnite?}
 QuitTitle {type=string, default=Quit}
 TakeControlAudio {type=SoundBase, default=/Game/Sounds/Fort_UI_Sounds/FrontEnd/DefaultSounds/PressedButtons/MSS_UI_Button_TakeControl_Pressed}
]
widgets {
 info {inheritedWidgetCount=12, parentClass=/Script/FortniteUI.FortRootViewportLayout_Athena, rootWidgetClass=&Overlay, widgetCount=13}
 widgets [
  Master Overlay {widgetClassPath=&Overlay, widget=Master Overlay}
  StateContentBox {
   widgetClassPath=&SizeBox
   bInherited=true
   bIsVariable=true
   parent=Master Overlay
   widget=StateContentBox
   slot {class=OverlaySlot, properties {horizontalAlignment=HAlign_Fill, verticalAlignment=VAlign_Fill}}
  }
  WidgetQueue_ModalContent {
   widgetClassPath=/Script/CommonUI.CommonActivatableWidgetQueue
   bInherited=true
   bIsVariable=true
   parent=Master Overlay
   widget=WidgetQueue_ModalContent
   properties {transitionFallbackStrategy=Previous}
   slot {class=OverlaySlot, properties {horizontalAlignment=HAlign_Fill, verticalAlignment=VAlign_Fill}}
  }
  AthenaProgressModal {
   widgetClassPath=/Game/UI/Foundation/Modals/ProgressModal/Athena_ProgressModal
   bInherited=true
   bIsVariable=true
   parent=Master Overlay
   widget=AthenaProgressModal
   properties {visibility=Collapsed}
   slot {class=OverlaySlot, properties {horizontalAlignment=HAlign_Fill, verticalAlignment=VAlign_Fill}}
  }
  ToastDisplayArea_ToastHandler {
   widgetClassPath=/Game/UI/Foundation/Toasts/ToastDisplayArea
   bInherited=true
   bIsVariable=true
   parent=Master Overlay
   widget=ToastDisplayArea_ToastHandler
   properties {elementTag {tagName=Athena.HideForCapture}, hUDElementTag {gameplayTags []}}
   slot {class=OverlaySlot, properties {horizontalAlignment=HAlign_Fill, verticalAlignment=VAlign_Fill}}
  }
  CilantroInjection {
   widgetClassPath=/Game/UI/Plugin/FortPluginWidgetInjectionSlot
   bInherited=true
   bIsVariable=true
   parent=Master Overlay
   widget=CilantroInjection
   properties {pluginId=Cilantro, pluginWidgetId=CilantroPopup, visibility=Collapsed}
   slot {class=OverlaySlot, properties {horizontalAlignment=HAlign_Fill, verticalAlignment=VAlign_Fill}}
  }
  Overlay_Dialog {
   widgetClassPath=&Overlay
   bInherited=true
   parent=Master Overlay
   widget=Overlay_Dialog
   slot {class=OverlaySlot, properties {horizontalAlignment=HAlign_Fill, verticalAlignment=VAlign_Fill}}
  }
  ConfirmationWindow_ConfirmationWidget {
   widgetClassPath=/Game/UI/Foundation/Modals/ConfirmationWindow/ConfirmationWindow
   bInherited=true
   bIsVariable=true
   parent=Master Overlay
   widget=ConfirmationWindow_ConfirmationWidget
   properties {bIsModal=true, visibility=Collapsed}
   slot {class=OverlaySlot, properties {horizontalAlignment=HAlign_Fill, verticalAlignment=VAlign_Fill}}
  }
  ErrorWindow_Errors {
   widgetClassPath=/Script/CommonUI.CommonLazyWidget
   bInherited=true
   bIsVariable=true
   parent=Master Overlay
   widget=ErrorWindow_Errors
   properties {widgetClass=/Game/UI/Foundation/Modals/ErrorWindow/ErrorWindow.ErrorWindow_C}
   slot {class=OverlaySlot, properties {horizontalAlignment=HAlign_Fill, verticalAlignment=VAlign_Fill}}
  }
  RestartModal {
   widgetClassPath=/Game/UI/Foundation/Modals/RestartClient/RestartClient
   bInherited=true
   bIsVariable=true
   parent=Master Overlay
   widget=RestartModal
   properties {activatedVisibility=Visible, bIsModal=true, bSetVisibilityOnActivated=true, bSetVisibilityOnDeactivated=true}
   slot {class=OverlaySlot, properties {horizontalAlignment=HAlign_Fill, verticalAlignment=VAlign_Fill}}
  }
  PinModal {
   widgetClassPath=&CommonLazyWidget
   bInherited=true
   bIsVariable=true
   parent=Master Overlay
   widget=PinModal
   properties {widgetClass=/Game/UI/Foundation/Modals/PasscodeModal/PinModal.PinModal_C}
   slot {class=OverlaySlot, properties {horizontalAlignment=HAlign_Fill, verticalAlignment=VAlign_Fill}}
  }
  ProgressModal_ControllerDisconnected {
   widgetClassPath=/Game/UI/Foundation/Modals/ControllerDisconnectedModal/Athena_ControllerDisconnectedModal
   bInherited=true
   bIsVariable=true
   parent=Master Overlay
   widget=ProgressModal_ControllerDisconnected
   properties {visibility=Collapsed}
   slot {class=OverlaySlot, properties {horizontalAlignment=HAlign_Fill, verticalAlignment=VAlign_Fill}}
  }
  UpdateRequiredModal {
   widgetClassPath=/FortniteCoreUI/UI/Login/UpdateRequired/UpdateRequiredMP
   bInherited=true
   bIsVariable=true
   parent=Master Overlay
   widget=UpdateRequiredModal
   properties {bSetVisibilityOnActivated=true, bSetVisibilityOnDeactivated=true, visibility=Collapsed}
   slot {class=OverlaySlot, properties {horizontalAlignment=HAlign_Fill, verticalAlignment=VAlign_Fill}}
  }
 ]
}
